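# Niagara shockwave effect for the GarlicWhisk emote, built from two mesh emitters.
# Wind spawns a short burst of randomly oriented meshes, pushes them outward along their X axis by emitter age, and drives time-of-day blue tints and material parameters with curves.
# Wind_Wave bursts a mesh sized by User.Scale that expands and fades.
# Wave particles outlive their lifetime until the execution state is inactive and the system Timeline stops playing.
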NiagaraSystem /BRCosmetics/Animation/Game/MainPlayer/Emotes/GarlicWhisk/FX/MS_Emote_GarlicWish_Shockwave
dependencies {
 usedDataInterfaces [
  &NiagaraDataInterfaceArrayColor
  &NiagaraDataInterfaceCurve
  &NiagaraDataInterfaceColorCurve
  &NiagaraDataInterfaceVectorCurve
  &NiagaraDataInterfaceArrayFloat
 ]
 usedDynamicInputs [&MakeVector4_FromLinearColor]
 usedModules [
  &self:SystemSpawnScript.SystemScriptSource.SystemScriptGraph.NiagaraNodeAssignment_1.SetVariables_#1
  &SystemState
  &Timeline
 ]
}
emitters {
 Wind {
  emitterUpdateScript [
   &EmitterState(Enable Distance Culling=true, MaxDistance=6000, Module.bUseMaxDistance=true, Module.bUseMinDistance=false, Scalability Mode=::NewEnumerator1)
   &SpawnRate(
    SpawnRate=&FloatFromCurve(
     FloatCurve=DataInterface() {propertyValues {Curve {keys [
         {arriveTangent=77.1005, interpMode=RCIM_Cubic, leaveTangent=77.1005, tangentMode=RCTM_User, time=0.05, value=0}
         <at 0.07 1 RCIM_Cubic>
         {arriveTangent=-14.3902, interpMode=RCIM_Cubic, leaveTangent=-14.3902, tangentMode=RCTM_User, time=0.25, value=0}
        ]}}}
     CurveIndex=$System.Age
     Scale Curve=40
    )
    Use Spawn Probability=false
   )
   &SpawnBurst_Instantaneous(Spawn Count=5, Spawn Time=0.05, Use Spawn Probability=false, UseLoopCountLimit=false)
  ]
  particleSpawnScript [
   &InitializeParticle(
    Color=<rgba 1 1 1 0.6>
    Lifetime=&UniformRangedFloat(Minimum=0.3, Maximum=0.47)
    Mesh Scale=&Lerp_Vector(A=<0.8681 0.8681 0.7812>, B=<1.1285 1.1285 2>, Alpha=&RandomRangeFloat(Minimum=0, Maximum=1))
    Mesh Scale Mode=::NewEnumerator4
    UsePositionOffset=false
    Write Material Random=true
   ) {module=InitializeParticle001}
   &InitialMeshOrientation(
    Mesh Axis To Orient=<0 0 1>
    Mesh Orientation Mode=::NewEnumerator2
    Orientation Vector=&RandomRangeVector(Minimum=<0 0 1>, Maximum=<0 0 1>)
    Rotation=&RandomRangeVector(Minimum=<0 0 0>, Maximum=<0 0 1>)
    Use Rotation Vector=true
   )
   &ShapeLocation(
    Box Size=<1 1 70>
    MaskSpawnGroup=false
    Override Seed=false
    Rotation Coordinate Space=::World
    Shape Primitive=::NewEnumerator2
    Surface Only=false
    UseOwnerScale=true
   )
  ]
  particleUpdateScript [
   &ParticleState()
   &ScaleMeshSize(Scale Factor=&VectorFromCurve(VectorCurve=DataInterface() {propertyValues {
       XCurve {keys [
         {arriveTangent=1.9177, interpMode=RCIM_Cubic, leaveTangent=1.9177, tangentMode=RCTM_User, time=0, value=0.1}
         {arriveTangent=0.2969, interpMode=RCIM_Cubic, leaveTangent=0.8713, tangentMode=RCTM_User, time=1, value=1}
        ]}
       YCurve {keys [
         {arriveTangent=1.9177, interpMode=RCIM_Cubic, leaveTangent=1.9177, tangentMode=RCTM_User, time=0, value=0.1}
         {arriveTangent=0.2969, interpMode=RCIM_Cubic, leaveTangent=0.5259, tangentMode=RCTM_User, time=1, value=1}
        ]}
       ZCurve {keys [
         {arriveTangent=2.8674, interpMode=RCIM_Cubic, leaveTangent=2.8674, tangentMode=RCTM_User, time=0, value=0.5}
         {arriveTangent=-0.4, interpMode=RCIM_Cubic, leaveTangent=-0.4, time=0.35, value=1}
         {arriveTangent=-1.6862, interpMode=RCIM_Cubic, leaveTangent=-1.6862, tangentMode=RCTM_User, time=1, value=0.1}
        ]}
      }}))
   /SetVariables(Particles.Position=&AddVectorToPosition(
     Position=&Multiply_VectorByFloat(
      Vector=&ReturnMeshOrientationAxis(Mesh Orientation Axis=::XAxis, Mesh Orientation Quat=$Particles.MeshOrientation)
      Float=&FloatFromCurve(
       FloatCurve=DataInterface() {propertyValues {Curve {keys [
           {arriveTangent=3.4662, interpMode=RCIM_Cubic, leaveTangent=3.4662, tangentMode=RCTM_User, time=0, value=-0.2}
           {interpMode=RCIM_Cubic, tangentMode=RCTM_User, time=0.5, value=1}
          ]}}}
       CurveIndex=$Emitter.Age
       Scale Curve=&Multiply_Float(A=&MakeFloatFromVector(Vector=$Particles.Initial.Scale, Channel=::NewEnumerator0), B=18)
      )
     )
     Vector=&MakeVector(X=0, Y=0, Z=&RandomRangeFloat(Minimum=25, Maximum=90))
    )) {isSetParameters=true}
   &CameraOffset(Camera Offset Amount=&Lerp_Float(A=70, B=140, Alpha=$Particles.NormalizedAge))
   &ScaleColor(
    Scale Alpha=&FloatFromCurve(FloatCurve=DataInterface() {propertyValues {Curve {keys [
         {arriveTangent=8.125, interpMode=RCIM_Cubic, time=0, value=0}
         {interpMode=RCIM_Cubic, tangentMode=RCTM_User, time=0.3, value=1}
         {arriveTangent=-0.8126, interpMode=RCIM_Cubic, leaveTangent=-0.8126, tangentMode=RCTM_User, time=0.7791, value=0.7}
         {interpMode=RCIM_Cubic, leaveTangent=-2.3395, time=1, value=0}
        ]}}})
    Scale RGB=&MakeVectorFromLinearColorRGB(Color=&FNTimeOfDayColor(
      Night Color=&ColorFromCurve(
       ColorCurve=DataInterface() {propertyValues {
         AlphaCurve {keys []}
         BlueCurve {keys [<at 0.0485 2 RCIM_Cubic>,<at 0.2577 1.5>]}
         GreenCurve {keys [<at 0.0485 0.1816 RCIM_Cubic>,<at 0.2577 0.2644>]}
         RedCurve {keys [<at 0.0485 0.1918 RCIM_Cubic>,<at 0.2577 0.2187>]}
        }}
       CurveIndex=$Emitter.Age
      )
      Morning Color=&RandomRangeLinearColor(
       Minimum=&ColorFromCurve(
        ColorCurve=DataInterface() {propertyValues {
          AlphaCurve {keys []}
          BlueCurve {keys [<at 0.0485 2 RCIM_Cubic>,<at 0.2577 1.5>]}
          GreenCurve {keys [<at 0.0485 0.7327 RCIM_Cubic>,<at 0.2577 0.3774>]}
          RedCurve {keys [<at 0.0485 0.6458 RCIM_Cubic>,<at 0.2577 0.3359>]}
         }}
        CurveIndex=$Emitter.Age
       )
       Maximum=&ColorFromCurve(
        ColorCurve=DataInterface() {propertyValues {
          AlphaCurve {keys []}
          BlueCurve {keys [<at 0.0485 2 RCIM_Cubic>,<at 0.2577 1.5>]}
          GreenCurve {keys [<at 0.0485 0.5768 RCIM_Cubic>,<at 0.2577 0.2493>]}
          RedCurve {keys [<at 0.0485 0.4792 RCIM_Cubic>,<at 0.2577 0.2031>]}
         }}
        CurveIndex=$Emitter.Age
       )
      )
      Day Color=&RandomRangeLinearColor(
       Minimum=&ColorFromCurve(
        ColorCurve=DataInterface() {propertyValues {
          AlphaCurve {keys []}
          BlueCurve {keys [<at 0.0485 2 RCIM_Cubic>,<at 0.2577 1.5>]}
          GreenCurve {keys [<at 0.0485 0.7327 RCIM_Cubic>,<at 0.2577 0.3774>]}
          RedCurve {keys [<at 0.0485 0.6458 RCIM_Cubic>,<at 0.2577 0.3359>]}
         }}
        CurveIndex=$Emitter.Age
       )
       Maximum=&ColorFromCurve(
        ColorCurve=DataInterface() {propertyValues {
          AlphaCurve {keys []}
          BlueCurve {keys [<at 0.0485 2 RCIM_Cubic>,<at 0.2577 1.5>]}
          GreenCurve {keys [<at 0.0485 0.5768 RCIM_Cubic>,<at 0.2577 0.2493>]}
          RedCurve {keys [<at 0.0485 0.4792 RCIM_Cubic>,<at 0.2577 0.2031>]}
         }}
        CurveIndex=$Emitter.Age
       )
      )
      Evening Color=&RandomRangeLinearColor(
       Minimum=&ColorFromCurve(
        ColorCurve=DataInterface() {propertyValues {
          AlphaCurve {keys []}
          BlueCurve {keys [<at 0.0485 2 RCIM_Cubic>,<at 0.2577 1.5>]}
          GreenCurve {keys [<at 0.0485 0.7327 RCIM_Cubic>,<at 0.2577 0.3774>]}
          RedCurve {keys [<at 0.0485 0.6458 RCIM_Cubic>,<at 0.2577 0.3359>]}
         }}
        CurveIndex=$Emitter.Age
       )
       Maximum=&ColorFromCurve(
        ColorCurve=DataInterface() {propertyValues {
          AlphaCurve {keys []}
          BlueCurve {keys [<at 0.0485 2 RCIM_Cubic>,<at 0.2577 1.5>]}
          GreenCurve {keys [<at 0.0485 0.5768 RCIM_Cubic>,<at 0.2577 0.2493>]}
          RedCurve {keys [<at 0.0485 0.4792 RCIM_Cubic>,<at 0.2577 0.2031>]}
         }}
        CurveIndex=$Emitter.Age
       )
      )
     ))
    ScaleA=true
    ScaleRGB=true
   )
   &DynamicMaterialParameters(
    Index 0 Param 1=&FloatFromCurve(FloatCurve=DataInterface() {propertyValues {Curve {keys [
         {arriveTangent=-13.4039, interpMode=RCIM_Cubic, leaveTangent=-13.4039, tangentMode=RCTM_User, time=0, value=1}
         <at 0.1 0>
         {arriveTangent=1.1111, interpMode=RCIM_Cubic, leaveTangent=1.1111, time=0.4, value=0}
         {interpMode=RCIM_Cubic, leaveTangent=16.4602, time=1, value=1}
        ]}}})
    Index 0 Param 2=&FloatFromCurve(
     FloatCurve=DataInterface() {propertyValues {Curve {keys [
         {arriveTangent=2.4843, interpMode=RCIM_Cubic, leaveTangent=2.4843, tangentMode=RCTM_User, time=0, value=0}
         {interpMode=RCIM_Cubic, leaveTangent=-0.8461, time=1, value=1}
        ]}}}
     Scale Curve=0.5
    )
    Index 0 Param 4=&FloatFromCurve(
     FloatCurve=DataInterface() {propertyValues {Curve {keys [
         {arriveTangent=0.2805, interpMode=RCIM_Cubic, time=0, value=1}
         {arriveTangent=2.4494, interpMode=RCIM_Cubic, leaveTangent=2.4494, tangentMode=RCTM_User, time=1, value=2}
        ]}}}
     Scale Curve=&RandomRangeFloat(Minimum=0.6, Maximum=0.8)
    )
    Module.Param0WriteEnabled=true
    Module.Param1WriteEnabled=true
    Module.Param2WriteEnabled=false
    Module.Param3WriteEnabled=true
   )
  ]
  rendererClasses [&NiagaraMeshRendererProperties]
  renderers [{rendererClass=&NiagaraMeshRendererProperties, rendererIndex=0}]
 }
 Wind_Wave {
  emitterSpawnScript [/SetVariables(
    Emitter.DarkColor:LinearColor=<rgba 0.0535 0.0836 0.5 1>
    Emitter.MainColor:LinearColor=&SelectLinearColorFromArray(
     Array Sampling Mode=::NewEnumerator0
     Direct Array Index=&RandomRangeInt(Minimum=0, Maximum=1)
     Color Selection Array=DataInterface() {propertyValues {ColorData [<rgba 0.0487 0.1098 1 1>, <rgba 0.0819 0.1782 1 1>]}}
     Random Seed=0
    )
   ) {isSetParameters=true}]
  emitterUpdateScript [
   &EmitterState()
   &SpawnBurst_Instantaneous(Use Spawn Probability=false, UseLoopCountLimit=true)
  ]
  particleSpawnScript [
   &PartitionParticles()
   &InitializeParticle(
    Color=<rgba 1 1 1 3>
    Lifetime=0.7
    Mesh Scale Mode=::NewEnumerator3
    Mesh Uniform Scale=&Multiply_Float(A=1.5, B=$User.Scale)
    Position Offset=<0 0 15>
    UsePositionOffset=true
    Write Material Random=true
   )
   &InitialMeshOrientation(
    Mesh Orientation Mode=::NewEnumerator2
    Orientation Vector=<0 0 1>
    Rotation=&RandomRangeVector(Minimum=<-5 0 0>, Maximum=<5 0 0>)
    Use Rotation Vector=true
   )
   &CameraOffset(Camera Offset Amount=5)
  ]
  particleUpdateScript [
   &ParticleState(Kill Particles When Lifetime Has Elapsed=false)
   &ScaleMeshSize(Scale Factor=&VectorFromFloat(Value=&FloatFromCurve(FloatCurve=DataInterface() {propertyValues {Curve {keys [
          {arriveTangent=4.4607, interpMode=RCIM_Cubic, leaveTangent=4.4607, tangentMode=RCTM_User, time=0, value=0}
          {arriveTangent=1.3333, interpMode=RCIM_Cubic, leaveTangent=1.3333, time=0.3, value=0.6818}
          {arriveTangent=1.1375, interpMode=RCIM_Cubic, leaveTangent=1.1375, tangentMode=RCTM_User, time=0.75, value=1}
         ]}}})))
   &DynamicMaterialParameters(
    Index 0 Param 1=&FloatFromCurve(FloatCurve=DataInterface() {propertyValues {Curve {keys [
         {arriveTangent=9.0725, interpMode=RCIM_Cubic, leaveTangent=9.0725, tangentMode=RCTM_User, time=0, value=0}
         {arriveTangent=-2.5249, interpMode=RCIM_Cubic, leaveTangent=-2.5249, tangentMode=RCTM_User, time=0.1408, value=0.9847}
         {arriveTangent=-0.0405, interpMode=RCIM_Cubic, leaveTangent=-0.0405, tangentMode=RCTM_User, time=1, value=0}
        ]}}})
    Index 0 Param 2=&FloatFromCurve(FloatCurve=DataInterface() {propertyValues {Curve {keys [{arriveTangent=2.105, interpMode=RCIM_Cubic, leaveTangent=2.105, tangentMode=RCTM_User, time=0.1, value=0.1}, <at 1 1>]}}})
    Index 0 Param 3=100
    Index 0 Param 4=100
    Module.Param0WriteEnabled=false
    Module.Param1WriteEnabled=false
    Module.Param2WriteEnabled=true
    Module.Param3WriteEnabled=true
   ) {enabled=false}
   &ScaleColor(
    Scale Alpha=&FloatFromCurve(
     FloatCurve=DataInterface() {propertyValues {Curve {keys [{interpMode=RCIM_Cubic, tangentMode=RCTM_User, time=0.5, value=1}, {arriveTangent=-1.2034, tangentMode=RCTM_User, time=1, value=0}]}}}
     Scale Curve=&Lerp_Float(A=1, B=1, Alpha=$System.Timeline.NormalizedTime)
    )
    ScaleA=true
    ScaleRGB=true
   ) {module=ScaleColor001}
   &KillParticles(Kill Particles=&Boolean_AndOperation(
     Bool1=&Boolean_NotOperation(Bool=/Game/Effects/Niagara/DynamicInputs/State/FN_IsExecutionStateActive())
     Bool2=&Boolean_NotOperation(Bool=$System.Timeline.IsPlaying)
    ))
   &ScaleColor(ScaleA=true, ScaleRGB=true)
  ]
  rendererClasses [&NiagaraMeshRendererProperties]
  renderers [{rendererClass=&NiagaraMeshRendererProperties, rendererIndex=0}]
 }
}
summary {
 systemName=MS_Emote_GarlicWish_Shockwave
 emitters [
  {emitterName=Wind_Wave, rendererClasses [&NiagaraMeshRendererProperties]}
  {emitterName=Wind, rendererClasses [&NiagaraMeshRendererProperties]}
 ]
 userVariables [
  User.bIsFrontEnd {type=NiagaraBool}
  User.bIsFrontEndPreview {type=NiagaraBool}
  User.Scale {type=NiagaraFloat, defaultValue=1, description=""}
  User.VortexHeight {type=Vector3f, description="", defaultValue {struct=Vector3f, value=<0 0 1>}}
 ]
}
