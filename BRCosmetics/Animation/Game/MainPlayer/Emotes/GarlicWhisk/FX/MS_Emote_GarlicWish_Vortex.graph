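# Niagara system for the vortex effect of the GarlicWhisk player emote: a blue, spinning column of ground meshes, airflow and smoke.
# Ground and airflow meshes spin and grow over loop age and the system Timeline, scaled by User.Scale, while GPU smoke sprites orbit and rise.
# Ground and smoke tints follow time of day, with smoke also varying by platform quality.
# Several emitters, including debris, stones, spirals and trail rings, are disabled.
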NiagaraSystem /BRCosmetics/Animation/Game/MainPlayer/Emotes/GarlicWhisk/FX/MS_Emote_GarlicWish_Vortex
dependencies {
 usedDataInterfaces [
  &NiagaraDataInterfaceCurve
  &NiagaraDataInterfaceArrayColor
  &NiagaraDataInterfaceColorCurve
  &NiagaraDataInterfaceVectorCurve
  &NiagaraDataInterfacePlatformSet
  &NiagaraDataInterfaceArrayFloat
 ]
 usedDynamicInputs [&MakeVector4_FromLinearColor, &ColorFromCurve]
 usedModules [
  &self:SystemSpawnScript.SystemScriptSource.SystemScriptGraph.NiagaraNodeAssignment_1.SetVariables_#1
  &SystemState
  &Timeline
 ]
}
emitters {
 Debris_01002 {
  bEnabled=false
  simTarget=GPUComputeSim
  emitterUpdateScript [
   &EmitterState(
    Enable Distance Culling=true
    MaxDistance=15000
    Module.bApplySpawnCountScaleByDistance=false
    Module.bUseMaxDistance=true
    Module.bUseMinDistance=false
    Scalability Mode=::NewEnumerator1
    Scale Spawn Count=true
    Spawn Count Scale By Distance=DataInterface() {propertyValues {Curve {keys [<at 0 1>,<at 0.25 1>,<at 1 0>]}}}
   )
   &SpawnRate(
    SpawnRate=&FloatFromCurve(FloatCurve=DataInterface() {propertyValues {Curve {keys [<at 0 0>,<at 0.2 100>,<at 0.7 0>]}}}, CurveIndex=$Emitter.NormalizedLoopAge)
    Use Spawn Probability=false
   )
  ]
  particleSpawnScript [
   &InitializeParticle(
    Color=<rgba 0 0 0 2>
    Lifetime=&RandomRangeFloat(Minimum=0.2, Maximum=0.7)
    Sprite Rotation Mode=::NewEnumerator2
    Sprite Size Mode=::NewEnumerator3
    Uniform Sprite Size=&FloatFromCurve(
     FloatCurve=DataInterface() {propertyValues {Curve {keys [
         {interpMode=RCIM_Cubic, tangentMode=RCTM_User, time=0, value=5}
         {arriveTangent=-4.5612, interpMode=RCIM_Cubic, leaveTangent=-4.5612, tangentMode=RCTM_User, time=1, value=0}
        ]}}}
     CurveIndex=&RandomRangeFloat(Minimum=0, Maximum=1)
    )
    UsePositionOffset=false
    Write Material Random=true
   )
   /SetVariables(Particles.SpriteSize=&Vector2DFromFloat(Value=&RandomRangeFloat(
      Minimum=3
      Maximum=&UniformAOrBFloat(A=10, B=&RandomRangeFloat(Minimum=20, Maximum=30), Distribution Weight=0.25)
     ))) {enabled=false, isSetParameters=true}
   /SetVariables(Particles.Color=&MultiplyLinearColor(Linear Color=$Output.FN_SpawnParticlesOnTerrainNearPlayer.LandscapeColor, Scale Factor=<1 1 1 1>)) {enabled=false, isSetParameters=true}
   &CameraOffset(Camera Offset Amount=10)
  ]
  particleUpdateScript [
   &ParticleState()
   &ShapeLocation(
    Cylinder Height=20
    Cylinder Height Midpoint=0
    Cylinder Radius=&RandomRangeFloat(Minimum=40, Maximum=60)
    MaskSpawnGroup=false
    Offset=&MakeVector(
     X=0
     Y=0
     Z=&FloatFromCurve(
      FloatCurve=DataInterface() {propertyValues {Curve {keys [<at 0 0>,<at 0 0>,<at 1 1>]}}}
      Scale Curve=&RandomRangeFloat(
       Minimum=0
       Maximum=&FloatFromCurve(FloatCurve=DataInterface() {propertyValues {Curve {keys [<at 0 0>,<at 1 150>]}}}, CurveIndex=$Emitter.NormalizedLoopAge)
      )
     )
    )
    Offset Mode=::NewEnumerator0
    Override Seed=false
    Rotation Angle:NiagaraFloat=&Multiply_Float(A=$Particles.Age, B=&RandomRangeFloat(Minimum=-0.1, Maximum=-0.5))
    Rotation Axis=<0 0 1>
    Rotation Mode=::NewEnumerator1
    Shape Primitive=::NewEnumerator1
    Surface Only=true
    Use Endcaps In Surface Only Mode=false
    UseOwnerScale=true
   )
   &DynamicMaterialParameters(
    Index 0 Param 1=0.5
    Index 0 Param 2=1
    Index 0 Param 3=&RandomRangeFloat(Minimum=500, Maximum=500)
    Index 0 Param 4=0.25
    Index 1 Param 1=512
    Module.Param0WriteEnabled=true
    Module.Param0WriteEnabled1=true
    Module.Param1WriteEnabled=true
    Module.Param1WriteEnabled1=true
    Module.Param2WriteEnabled=true
    Module.Param2WriteEnabled1=true
    Module.Param3WriteEnabled=true
    Module.Param3WriteEnabled1=true
    Write Parameter Index 1=true
   )
   &PointForce(Force Falloff Distance=5000, Force Strength=0, Module.Use Falloff=true, Module.UseLinearForce=false)
   &SolveForcesAndVelocity(Clamp Velocity=true, Limit Acceleration=false, Speed Limit=1500)
   &ScaleSpriteSize(Uniform Curve Sprite Scale=DataInterface() {propertyValues {Curve {keys [
        {interpMode=RCIM_Cubic, tangentMode=RCTM_User, time=0, value=0}
        {interpMode=RCIM_Cubic, tangentMode=RCTM_User, time=0.1, value=1}
        {interpMode=RCIM_Cubic, tangentMode=RCTM_User, time=0.8, value=1}
        {tangentMode=RCTM_User, time=1, value=0.05}
       ]}}}) {module=ScaleSpriteSize001}
   /Niagara/Modules/Update/Color/ScaleColorBySpeed(A Scale Min=2, Max Speed Threshold=1200, Min Speed Threshold=500, RGB Scale Min=1) {enabled=false}
   &SubUVAnimation(Start Frame:NiagaraInt32=&RandomRangeInt(Minimum=0, Maximum=5), SubUV Loop Count=0.5)
  ]
  rendererClasses [&NiagaraSpriteRendererProperties]
  renderers [
   {rendererClass=&NiagaraSpriteRendererProperties}
   {rendererClass=&NiagaraSpriteRendererProperties, rendererIndex=1}
  ]
 }
 Ground_FX {
  emitterSpawnScript [
   /SetVariables(Emitter.RotationSpeed:NiagaraFloat=&FloatFromCurve(FloatCurve=DataInterface() {propertyValues {Curve {keys [<at 0 -0.5>,<at 0.1 0>]}}}, CurveIndex=$Emitter.NormalizedLoopAge, Scale Curve=0)) {isSetParameters=true}
   /SetVariables(
    Emitter.DarkColor:LinearColor=&FNTimeOfDayColor(
     Night Color=<rgba 0.0572 0.1284 0.6458 1>
     Morning Color=<rgba 0.111 0.1757 0.6458 1>
     Day Color=<rgba 0.111 0.1757 0.6458 1>
     Evening Color=<rgba 0.111 0.1757 0.6458 1>
    )
    Emitter.MainColor:LinearColor=&FNTimeOfDayColor(
     Night Color=&SelectLinearColorFromArray(
      Array Sampling Mode=::NewEnumerator0
      Direct Array Index=&RandomRangeInt(Minimum=0, Maximum=1)
      Color Selection Array=DataInterface() {propertyValues {ColorData [<rgba 0 0.0088 1 1>]}}
      Random Seed=0
     )
     Morning Color=&SelectLinearColorFromArray(
      Array Sampling Mode=::NewEnumerator0
      Direct Array Index=&RandomRangeInt(Minimum=0, Maximum=1)
      Color Selection Array=DataInterface() {propertyValues {ColorData [<rgba 0.2083 0.2153 1 1>]}}
      Random Seed=0
     )
     Day Color=&SelectLinearColorFromArray(
      Array Sampling Mode=::NewEnumerator0
      Direct Array Index=&RandomRangeInt(Minimum=0, Maximum=1)
      Color Selection Array=DataInterface() {propertyValues {ColorData [<rgba 0.3073 0.1458 1 1>, <rgba 0.4752 0.4836 1 1>]}}
      Random Seed=0
     )
     Evening Color=&SelectLinearColorFromArray(
      Array Sampling Mode=::NewEnumerator0
      Direct Array Index=&RandomRangeInt(Minimum=0, Maximum=1)
      Color Selection Array=DataInterface() {propertyValues {ColorData [<rgba 0.2083 0.2153 1 1>]}}
      Random Seed=0
     )
    )
   ) {isSetParameters=true}
  ]
  emitterUpdateScript [
   &EmitterState()
   &SpawnRate(SpawnRate=8, Use Spawn Probability=true)
   &SpawnBurst_Instantaneous(Spawn Count=2, Use Spawn Probability=false, UseLoopCountLimit=false)
  ]
  particleSpawnScript [
   &InitializeParticle(
    Color=<rgba 1 1 1 3>
    Lifetime=&RandomRangeFloat(Minimum=0.6, Maximum=0.9)
    Mesh Scale=&Multiply_VectorByFloat(Vector=<0.1562 0.1562 0.45>, Float=$User.Scale)
    Mesh Scale Mode=::NewEnumerator4
    Position Offset Coordinate Space=::Local
    UsePositionOffset=true
    Write Material Random=true
   )
   &InitialMeshOrientation(
    Mesh Axis To Orient=<0 0 3>
    Mesh Orientation Mode=::NewEnumerator2
    Orientation Coordinate Space=::World
    Orientation Vector=<0 0 1>
    Rotation=&RandomRangeVector(Minimum=<0.03 0.03 0>, Maximum=<-0.03 -0.03 1>)
    Use Rotation Vector=true
   )
   &InitialMeshOrientation(Mesh Orientation Mode=::NewEnumerator7, Rotation=<0 0.1 0>, Use Rotation Vector=true) {enabled=false, module=InitialMeshOrientation001}
   &DynamicMaterialParameters(
    Index 0 Param 1=1
    Index 0 Param 2=1
    Index 0 Param 3=&RandomRangeFloat(Minimum=-1, Maximum=1)
    Index 0 Param 4=&RandomRangeFloat(Minimum=-1, Maximum=1)
    Module.Index0Param0Write=false
    Module.Index0Param1Write=false
    Module.Index0Param2Write=true
    Module.Index0Param3Write=true
   )
   /SetVariables(Particles.ScaleDetermination:NiagaraFloat=&MakeCustomFloatFromBool(
     Bool=&SetBoolByFloatComparison(A=&RandomRangeFloat(Minimum=0, Maximum=1), B=0.8)
     True Float=0.4
     False Float=1
    )) {isSetParameters=true}
   /SetVariables(Particles.float:NiagaraFloat=&FloatFromCurve(
     FloatCurve=DataInterface() {propertyValues {Curve {keys [
         <at 0 0 RCIM_Cubic>
         {arriveTangent=3.0002, interpMode=RCIM_Cubic, leaveTangent=3.0002, tangentMode=RCTM_User, time=0.7, value=1}
         {arriveTangent=0.1133, interpMode=RCIM_Cubic, leaveTangent=0.1133, tangentMode=RCTM_User, time=0.95, value=1}
         <at 1 0.4 RCIM_Cubic>
        ]}}}
     CurveIndex=$Emitter.NormalizedLoopAge
     Scale Curve=$Particles.ScaleDetermination
    )) {isSetParameters=true}
  ]
  particleUpdateScript [
   &ParticleState()
   &ScaleMeshSize(Scale Factor=&MakeVector(
     X=1
     Y=1
     Z=&Lerp_Float(
      A=0
      B=&FloatFromCurve(
       FloatCurve=DataInterface() {propertyValues {Curve {keys [
           <at 0 0 RCIM_Cubic>
           {arriveTangent=3.0002, interpMode=RCIM_Cubic, leaveTangent=3.0002, tangentMode=RCTM_User, time=0.57, value=1}
           {arriveTangent=0.1133, interpMode=RCIM_Cubic, leaveTangent=0.1133, tangentMode=RCTM_User, time=0.95, value=1}
           <at 1 0.4 RCIM_Cubic>
          ]}}}
       CurveIndex=$Emitter.NormalizedLoopAge
       Scale Curve=$Particles.ScaleDetermination
      )
      Alpha=&FloatFromCurve(FloatCurve=DataInterface() {propertyValues {Curve {keys [{arriveTangent=3.0906, interpMode=RCIM_Cubic, leaveTangent=3.0906, tangentMode=RCTM_User, time=0, value=0}, <at 0.7 1 RCIM_Cubic>]}}})
     )
    )) {module=ScaleMeshSize001}
   &ScaleMeshSize(Scale Factor=&Lerp_Vector(
     A=<1 1 0.5>
     B=<5 5 1.25>
     Alpha=&MakeCustomFloatFromBool(
      Bool=$System.Timeline.IsRewinding
      True Float=&FloatFromCurve(
       FloatCurve=DataInterface() {propertyValues {Curve {keys [{tangentMode=RCTM_User, time=0, value=1}, {tangentMode=RCTM_User, time=1, value=1.5}]}}}
       CurveIndex=$System.Timeline.InverseNormalizedTime
      )
      False Float=&FloatFromCurve(
       FloatCurve=DataInterface() {propertyValues {Curve {keys [{tangentMode=RCTM_User, time=0, value=0}, {tangentMode=RCTM_User, time=1, value=1}]}}}
       CurveIndex=$System.Timeline.NormalizedTime
      )
     )
    ))
   &UpdateMeshOrientation(
    Rotation Rate=&FloatFromCurve(
     FloatCurve=DataInterface() {propertyValues {Curve {keys [<at 0.55 0 RCIM_Cubic>, {arriveTangent=2.9123, interpMode=RCIM_Cubic, leaveTangent=2.9123, tangentMode=RCTM_User, time=1, value=1}]}}}
     CurveIndex=$Emitter.NormalizedLoopAge
     Scale Curve=-1.5
    )
    Rotation Vector=<0 0 1>
   )
   &ScaleColor(
    Scale Alpha=&FloatFromCurve(FloatCurve=DataInterface() {propertyValues {Curve {keys [
         {interpMode=RCIM_Cubic, tangentMode=RCTM_User, time=0, value=1}
         {interpMode=RCIM_Cubic, tangentMode=RCTM_User, time=0.5, value=1}
         {tangentMode=RCTM_User, time=1, value=0}
        ]}}})
    Scale RGB=<5 5 5>
    ScaleA=true
    ScaleRGB=true
   )
   &DynamicMaterialParameters(
    Index 0 Param 1=&FloatFromCurve(FloatCurve=DataInterface() {propertyValues {Curve {keys [{tangentMode=RCTM_User, time=0.4, value=1}, {tangentMode=RCTM_User, time=0.6, value=-1}]}}})
    Index 0 Param 2=1
    Index 0 Param 3=&RandomRangeFloat(Minimum=-1, Maximum=1)
    Index 0 Param 4=&RandomRangeFloat(Minimum=-1, Maximum=1)
    Index 1 Param 1=&FloatFromCurve(
     FloatCurve=DataInterface() {propertyValues {Curve {keys [
         {arriveTangent=1.7366, interpMode=RCIM_Cubic, leaveTangent=1.7366, tangentMode=RCTM_User, time=0, value=0}
         {arriveTangent=0.4174, interpMode=RCIM_Cubic, leaveTangent=0.4174, tangentMode=RCTM_User, time=1, value=1}
        ]}}}
     Scale Curve=0.5
    )
    Index 1 Param 2=1
    Module.Index0Param0Write=true
    Module.Index0Param1Write=false
    Module.Index0Param2Write=true
    Module.Index0Param3Write=true
    Module.Index1Param0Write=true
    Module.Index1Param1Write=false
    Module.Index1Param2Write=false
    Module.Index1Param3Write=false
    Write Parameter Index 1=true
   ) {module=DynamicMaterialParameters001}
   &ScaleColor(
    Scale Alpha=&FloatFromCurve(
     FloatCurve=DataInterface() {propertyValues {Curve {keys [<at 0.1 0>,<at 0.2 1>,<at 0.8 1>,<at 0.85 0>]}}}
     CurveIndex=$Emitter.NormalizedLoopAge
    )
    Scale RGB=&VectorFromFloat(Value=&FloatFromCurve(
      FloatCurve=DataInterface() {propertyValues {Curve {keys [<at 0.1 0>,<at 0.2 1>,<at 0.7 1>,<at 0.9 0>]}}}
      CurveIndex=$Emitter.NormalizedLoopAge
     ))
    ScaleA=true
    ScaleRGB=true
   ) {module=ScaleColor001}
   &KillParticles(Kill Particles=&SetBoolByFloatComparison(A=$Emitter.NormalizedLoopAge, B=0.92))
  ]
  rendererClasses [&NiagaraMeshRendererProperties]
  renderers [{rendererClass=&NiagaraMeshRendererProperties, rendererIndex=0}]
 }
 Ground_FX001 {
  bEnabled=false
  emitterSpawnScript [/SetVariables(
    Emitter.DarkColor:LinearColor=<rgba 0.1629 0.1924 0.4062 1>
    Emitter.MainColor:LinearColor=&SelectLinearColorFromArray(
     Array Sampling Mode=::NewEnumerator0
     Direct Array Index=&RandomRangeInt(Minimum=0, Maximum=1)
     Color Selection Array=DataInterface() {propertyValues {ColorData [<rgba 0.0487 0.1098 1 1>, <rgba 0.0819 0.1782 1 1>]}}
     Random Seed=0
    )
   ) {isSetParameters=true}]
  emitterUpdateScript [
   &EmitterState(Life Cycle Mode=::NewEnumerator1, Loop Duration=4.5, UseLoopDelay=false)
   &SpawnRate(SpawnRate=0.8, Use Spawn Probability=false) {enabled=false}
   &SpawnBurst_Instantaneous(Use Spawn Probability=false, UseLoopCountLimit=false)
  ]
  particleSpawnScript [
   /SetVariables(Particles.ScaleDetermination:NiagaraFloat=&MakeCustomFloatFromBool(
     Bool=&SetBoolByFloatComparison(A=&RandomRangeFloat(Minimum=0, Maximum=1), B=0.8)
     True Float=0.4
     False Float=1
    )) {isSetParameters=true}
   &InitializeParticle(
    Color=<rgba 1 1 1 0.5>
    Lifetime=4
    Mesh Scale=&Multiply_VectorByFloat(
     Vector=&Lerp_Vector(A=<1.125 1.125 3.5>, B=<1.125 1.125 3.5>, Alpha=$Particles.ScaleDetermination)
     Float=$User.Scale
    )
    Mesh Scale Mode=::NewEnumerator4
    UsePositionOffset=false
    Write Material Random=true
   )
   &InitialMeshOrientation(
    Mesh Axis To Orient=&Add_Vector(A=<0 0 3>, B=<0 0 0>)
    Mesh Orientation Mode=::NewEnumerator2
    Orientation Coordinate Space=::World
    Orientation Vector=<0 0 1>
    Rotation=&RandomRangeVector(Minimum=<0.02 0.02 0>, Maximum=<-0.02 -0.02 1>)
    Use Rotation Vector=true
   )
   &InitialMeshOrientation(Mesh Orientation Mode=::NewEnumerator7, Rotation=<0 0.1 0>, Use Rotation Vector=true) {enabled=false, module=InitialMeshOrientation001}
   &AddRotationalVelocity(Rotation Rate=&MakeVector(
     X=0
     Y=0
     Z=&FloatFromCurve(FloatCurve=DataInterface() {propertyValues {Curve {keys [<at 0 -2>,<at 0.5 -2>]}}}, CurveIndex=$Emitter.NormalizedLoopAge, Scale Curve=0.3)
    ))
   &DynamicMaterialParameters(
    Index 0 Param 1=1
    Index 0 Param 2=1
    Index 0 Param 3=&RandomRangeFloat(Minimum=-1, Maximum=1)
    Index 0 Param 4=&RandomRangeFloat(Minimum=-1, Maximum=1)
    Module.Index0Param0Write=false
    Module.Index0Param1Write=false
    Module.Index0Param2Write=true
    Module.Index0Param3Write=true
   )
  ]
  particleUpdateScript [
   &SolveForcesAndVelocity(Clamp Velocity=false, Limit Acceleration=false)
   &ParticleState()
   &ScaleMeshSize(Scale Factor=&MakeVector(
     X=1
     Y=1
     Z=&FloatFromCurve(
      FloatCurve=DataInterface() {propertyValues {Curve {keys [
          <at 0 0 RCIM_Cubic>
          {arriveTangent=1.8679, interpMode=RCIM_Cubic, leaveTangent=1.8679, tangentMode=RCTM_User, time=0.7, value=1}
          {arriveTangent=0.1133, interpMode=RCIM_Cubic, leaveTangent=0.1133, tangentMode=RCTM_User, time=0.9, value=1}
          <at 1 0.6 RCIM_Cubic>
         ]}}}
      CurveIndex=$Emitter.NormalizedLoopAge
     )
    )) {module=ScaleMeshSize001}
   &ScaleMeshSize(Scale Factor=&Lerp_Vector(
     A=<0.2 0.2 0.33>
     B=<1 1 1>
     Alpha=&MakeCustomFloatFromBool(
      Bool=$System.Timeline.IsRewinding
      True Float=&FloatFromCurve(
       FloatCurve=DataInterface() {propertyValues {Curve {keys [{tangentMode=RCTM_User, time=0, value=1}, {tangentMode=RCTM_User, time=1, value=1.5}]}}}
       CurveIndex=$System.Timeline.InverseNormalizedTime
      )
      False Float=&FloatFromCurve(
       FloatCurve=DataInterface() {propertyValues {Curve {keys [{tangentMode=RCTM_User, time=0, value=0}, {tangentMode=RCTM_User, time=1, value=1}]}}}
       CurveIndex=$System.Timeline.NormalizedTime
      )
     )
    ))
   &UpdateMeshOrientation(Rotation Rate=-0.5, Rotation Vector=<0 1 0>) {enabled=false}
   &ScaleColor(
    Scale Alpha=&FloatFromCurve(
     FloatCurve=DataInterface() {propertyValues {Curve {keys [
         {interpMode=RCIM_Cubic, tangentMode=RCTM_User, time=0, value=0}
         {interpMode=RCIM_Cubic, tangentMode=RCTM_User, time=0.5, value=1}
         {tangentMode=RCTM_User, time=1, value=0}
        ]}}}
     Scale Curve=&Lerp_Float(A=0, B=0.5, Alpha=$System.Timeline.NormalizedTime)
    )
    ScaleA=true
    ScaleRGB=true
   )
   &DynamicMaterialParameters(
    Index 1 Param 1=&FloatFromCurve(
     FloatCurve=DataInterface() {propertyValues {Curve {keys [
         {arriveTangent=1.7366, interpMode=RCIM_Cubic, leaveTangent=1.7366, tangentMode=RCTM_User, time=0, value=0}
         {arriveTangent=0.4174, interpMode=RCIM_Cubic, leaveTangent=0.4174, tangentMode=RCTM_User, time=1, value=1}
        ]}}}
     Scale Curve=0.5
    )
    Index 1 Param 2=1
    Module.Index1Param0Write=true
    Module.Index1Param1Write=false
    Module.Index1Param2Write=false
    Module.Index1Param3Write=false
    Write Parameter Index 0=false
    Write Parameter Index 1=true
   ) {module=DynamicMaterialParameters001}
   &ScaleColor(
    Scale Alpha=&FloatFromCurve(FloatCurve=DataInterface() {propertyValues {Curve {keys [<at 0.5 1>,<at 1 1>,<at 3.5 1>,<at 3.9 0>]}}}, CurveIndex=$Emitter.Age)
    ScaleA=true
    ScaleRGB=true
   ) {module=ScaleColor001}
  ]
  rendererClasses [&NiagaraMeshRendererProperties]
  renderers [{rendererClass=&NiagaraMeshRendererProperties, rendererIndex=0}]
 }
 NS_SoftAirflow {
  emitterSpawnScript [/SetVariables(Emitter.DarkColor:LinearColor=<rgba 0 0.0782 0.6458 1>) {isSetParameters=true}]
  emitterUpdateScript [
   &EmitterState(Life Cycle Mode=::NewEnumerator1, Loop Behavior=::NewEnumerator1, Loop Duration=3.3, UseLoopDelay=false)
   &SpawnRate(SpawnRate=2, Use Spawn Probability=false) {module=SpawnRate001}
   &SpawnRate(
    SpawnRate=&FloatFromCurve(FloatCurve=DataInterface() {propertyValues {Curve {keys [<at 0 1>,<at 0.95 0>]}}}, CurveIndex=$Emitter.NormalizedLoopAge, Scale Curve=12)
    Use Spawn Probability=false
   ) {enabled=false}
   &SpawnBurst_Instantaneous(Spawn Count=5, Use Spawn Probability=false, UseLoopCountLimit=false) {enabled=false}
  ]
  particleSpawnScript [
   &InitializeParticle(
    Color=&SelectLinearColorFromArray(
     Array Sampling Mode=::NewEnumerator0
     Direct Array Index=?
     Color Selection Array=DataInterface() {propertyValues {ColorData [<rgba 0.0865 0.1419 1 1>, <rgba 0.0772 0.265 1 1>, <rgba 0.5455 0.6185 1 1.5>, <rgba 0.0772 0.265 1 1>, <rgba 0.0865 0.1419 1 1>]}}
     Random Seed=0
    )
    Lifetime Max=2.5
    Lifetime Min=1.5
    Lifetime Mode=::NewEnumerator1
    Mesh Scale=&Multiply_VectorByFloat(Vector=<1 1 1.2>, Float=&RandomRangeFloat(Minimum=1, Maximum=1.6667))
    Mesh Scale Mode=::NewEnumerator4
    Position Offset=<0 0 200>
    UsePositionOffset=false
    Write Material Random=true
   )
   &InitialMeshOrientation(
    Mesh Orientation Mode=::NewEnumerator7
    Rotation=&RandomRangeVector(Minimum=<0 0 -1>, Maximum=<0 0 1>)
    Use Rotation Vector=true
   )
   &ShapeLocation(
    Cylinder Height=0
    Cylinder Height Midpoint=0
    Cylinder Radius=0
    MaskSpawnGroup=false
    Non Uniform Scale=&Multiply_VectorByFloat(Vector=<1 1 1>, Float=$User.Scale)
    Override Seed=false
    Shape Primitive=::NewEnumerator1
    Surface Only=false
    UseOwnerScale=true
   )
   &AddVelocity(Velocity=<0 0 100>, Velocity Speed Scale=500) {enabled=false}
   &ApplyInitialForces() {enabled=false}
   /SetVariables(Particles.ZGoalpost:NiagaraFloat=&FloatFromCurve(
     FloatCurve=DataInterface() {propertyValues {Curve {keys [{tangentMode=RCTM_User, time=0, value=0}, {tangentMode=RCTM_User, time=1, value=300}]}}}
     CurveIndex=$Emitter.NormalizedLoopAge
    )) {isSetParameters=true}
  ]
  particleUpdateScript [
   &ParticleState()
   /SetVariables(Particles.Position=&Lerp_Vector(
     A=Linked() {linkedVariable {name=Particles.Initial.Position, type=NiagaraPosition}}
     B=&MakeVector(X=0, Y=0, Z=$Particles.Initial.ZGoalpost)
     Alpha=&FloatFromCurve(FloatCurve=DataInterface() {propertyValues {Curve {keys [{tangentMode=RCTM_User, time=0, value=0}, {tangentMode=RCTM_User, time=1, value=1}]}}})
    )) {isSetParameters=true}
   &DynamicMaterialParameters(
    Index 0 Param 1=&FloatFromCurve(FloatCurve=DataInterface() {propertyValues {Curve {keys [{interpMode=RCIM_Cubic, tangentMode=RCTM_User, time=0.5, value=0}, {interpMode=RCIM_Cubic, tangentMode=RCTM_User, time=1, value=1}]}}})
    Index 0 Param 3=1
    Module.Param0WriteEnabled=true
    Module.Param1WriteEnabled=true
    Module.Param2WriteEnabled=true
    Module.Param3WriteEnabled=true
   )
   &ScaleMeshSize(Scale Factor=&Multiply_VectorByFloat(
     Vector=<1 1 1.5>
     Float=&FloatFromCurve(
      FloatCurve=DataInterface() {propertyValues {Curve {keys [
          {arriveTangent=4.1785, interpMode=RCIM_Cubic, leaveTangent=4.1785, tangentMode=RCTM_User, time=0, value=0}
          {arriveTangent=1.1179, interpMode=RCIM_Cubic, leaveTangent=1.1179, tangentMode=RCTM_User, time=0.2042, value=0.6228}
          {arriveTangent=0.0224, interpMode=RCIM_Cubic, leaveTangent=0.0224, tangentMode=RCTM_User, time=1, value=1}
         ]}}}
      Scale Curve=&Multiply_Float(A=1, B=$User.Scale)
     )
    ))
   &UpdateMeshOrientation(
    Rotation Rate=&FloatFromCurve(FloatCurve=DataInterface() {propertyValues {Curve {keys [<at 0 0>,<at 1 1>]}}}, Scale Curve=-2)
    Rotation Vector=<0 0 1>
   )
   &ScaleColor(
    Scale Alpha=&FloatFromCurve(
     FloatCurve=DataInterface() {propertyValues {Curve {keys [
         {arriveTangent=1.8077, interpMode=RCIM_Cubic, leaveTangent=1.8077, tangentMode=RCTM_User, time=0, value=0}
         {arriveTangent=-1.9397, interpMode=RCIM_Cubic, leaveTangent=-2.9533, tangentMode=RCTM_Break, time=0.5, value=0.5}
         {arriveTangent=-0.7636, interpMode=RCIM_Cubic, leaveTangent=-0.7636, tangentMode=RCTM_User, time=1, value=0}
        ]}}}
     Scale Curve=&Lerp_Float(A=0, B=0.25, Alpha=$System.Timeline.NormalizedTime)
    )
    ScaleA=true
    ScaleRGB=true
   )
  ]
  rendererClasses [&NiagaraMeshRendererProperties]
  renderers [{rendererClass=&NiagaraMeshRendererProperties, rendererIndex=0}]
 }
 Smoke {
  simTarget=GPUComputeSim
  emitterUpdateScript [
   &EmitterState()
   &SpawnRate(
    SpawnRate=&FloatFromCurve(
     FloatCurve=DataInterface() {propertyValues {Curve {keys [<at 0 0>,<at 0.1 1>,<at 0.7 1>]}}}
     CurveIndex=$Emitter.NormalizedLoopAge
     Scale Curve=8
    )
    Use Spawn Probability=false
   )
  ]
  particleSpawnScript [
   /SetVariables(Particles.VisibilityTag=1) {isSetParameters=true}
   &InitializeParticle(
    Color=&MakeCustomLinearColorFromBool(
     Bool=&IsPlatformSetActive(New Platform Set=DataInterface() {propertyValues {Platforms {qualityLevelMask=-29}}})
     True Color=&FNTimeOfDayColor(
      Night Color=<rgba 0.0285 0.076 0.2188 0.15>
      Morning Color=<rgba 0.0475 0.1267 0.3646 0.4>
      Day Color=&RandomRangeLinearColor(Minimum=<rgba 0.0745 0.0696 0.2344 1>, Maximum=<rgba 0.0891 0.0892 0.2344 1>)
      Evening Color=<rgba 0.0571 0.1223 0.3177 0.1>
     )
     False Color=&FNTimeOfDayColor(
      Night Color=&RandomRangeLinearColor(Minimum=<rgba 0.0863 0.0818 0.2344 1>, Maximum=<rgba 0.0549 0.055 0.2344 1>)
      Morning Color=<rgba 0.0475 0.1267 0.3646 0.8>
      Day Color=&RandomRangeLinearColor(Minimum=<rgba 0.0745 0.0696 0.2344 1>, Maximum=<rgba 0.0891 0.0892 0.2344 1>)
      Evening Color=<rgba 0.0571 0.1223 0.3177 0.6>
     )
    )
    Lifetime=&RandomRangeFloat(Minimum=0.8, Maximum=1.2)
    Mesh Scale Mode=::NewEnumerator3
    Sprite Rotation Mode=::NewEnumerator2
    Sprite Size Mode=::NewEnumerator3
    Uniform Sprite Size=&RandomRangeFloat(Minimum=86, Maximum=120)
    UsePositionOffset=false
    Write Material Random=true
   )
   &TorusLocation(Align Torus With Axis=false, MaskSpawnGroup=false, Module.Override Randomness=false, Torus Mode=::NewEnumerator2)
   &SphereLocation(
    Hemisphere Z=true
    MaskSpawnGroup=false
    Non Uniform Scale=<0.25 0.25 1>
    Override Seed=false
    Sphere Radius=15
    Surface Only=false
    UseNonUniformScale=true
    UseOffset=false
    UseSphereOrientation=false
   )
   /SetVariables(Particles.InitialPos:NiagaraPosition=$Particles.Position) {isSetParameters=true}
   &AddVelocity()
   /SetVariables(Particles.OutsideCOlor:LinearColor=$Output.FN_SpawnParticlesOnTerrainNearPlayer.LandscapeColor) {isSetParameters=true}
  ]
  particleUpdateScript [
   &ParticleState()
   /SetVariables(Particles.VisibilityTag=&MakeCustomIntFromBool(Bool=$Transient.FirstFrame, True INT=1, False INT=0)) {isSetParameters=true}
   &VortexForce(
    Influence Falloff Exponent=0.25
    Influence Falloff Radius=5000
    Origin Pull Amount=250
    Use Influence Falloff=true
    UseOriginOffset=true
    Vortex Force Amount=-500
    Vortex Origin Offset:Vector3f=<0 0 1000>
   ) {enabled=false}
   &RotateAroundPoint(
    Radius:NiagaraFloat=&Multiply_Float(
     A=&RandomRangeFloat(Minimum=0, Maximum=60)
     B=&FloatFromCurve(
      FloatCurve=DataInterface() {propertyValues {Curve {keys [<at 0 1>,<at 1 1.5>]}}}
      Scale Curve=&RandomRangeFloat(Minimum=0.8, Maximum=1.2)
     )
    )
    Rotation Center:NiagaraPosition=&MakeVector(
     X=0
     Y=0
     Z=&FloatFromCurve(
      FloatCurve=DataInterface() {propertyValues {Curve {keys [<at 0 0>,<at 1 35>]}}}
      Scale Curve=&RandomRangeFloat(Minimum=0.5, Maximum=1.1)
     )
    )
    Rotation Rate:NiagaraFloat=&Multiply_Float(
     A=&RandomRangeFloat(Minimum=-0.6, Maximum=-1.2)
     B=&FloatFromCurve(
      FloatCurve=DataInterface() {propertyValues {Curve {keys [{tangentMode=RCTM_User, time=0, value=1}, {tangentMode=RCTM_User, time=1, value=1.5}]}}}
      CurveIndex=$Emitter.NormalizedLoopAge
     )
    )
   )
   &GravityForce(Gravity=<0 0 20>)
   &PointAttractionForce(
    Attraction Radius=2500
    Attraction Strength=15
    Attractor Position Offset:Vector3f=<0 0 50>
    Kill Within Radius=false
    Module.Use Falloff=true
    UseAttractorOffset=true
   )
   &SolveForcesAndVelocity(Clamp Velocity=true, Limit Acceleration=true, Speed Limit=850)
   &ScaleColor(
    Scale Alpha=&FloatFromCurve(FloatCurve=DataInterface() {propertyValues {Curve {keys [
         {interpMode=RCIM_Cubic, tangentMode=RCTM_User, time=0, value=0}
         {interpMode=RCIM_Cubic, tangentMode=RCTM_User, time=0.5, value=1}
         <at 0.8 0.95>
         {tangentMode=RCTM_User, time=1, value=0}
        ]}}})
    Scale RGB=&MakeVectorFromLinearColorRGB(Color=<rgba 0 0 0 1>)
    ScaleA=true
    ScaleRGB=false
   )
   &ScaleColor(
    Scale Alpha=&FloatFromCurve(
     FloatCurve=DataInterface() {propertyValues {Curve {keys [{interpMode=RCIM_Cubic, tangentMode=RCTM_User, time=0, value=1}, <at 0.7 1>, {tangentMode=RCTM_User, time=0.9, value=0}]}}}
     CurveIndex=$Emitter.NormalizedLoopAge
     Scale Curve=2
    )
    Scale RGB=&MakeVectorFromLinearColorRGB(Color=<rgba 0 0 0 1>)
    ScaleA=true
    ScaleRGB=false
   ) {module=ScaleColor001}
   &SpriteRotationRate(Rotation Rate=&UniformRangedFloat(Minimum=-0.1, Maximum=0.1))
   &SubUVAnimation(
    Start Frame:NiagaraInt32=&RandomRangeInt(Minimum=12, Maximum=20)
    SubUV Animation Mode:ENiagara_SubUVLookupMode=::NewEnumerator1
   )
   &DynamicMaterialParameters(
    Index 0 Param 1=250
    Module.Param0WriteEnabled=true
    Module.Param1WriteEnabled=true
    Module.Param2WriteEnabled=true
    Module.Param3WriteEnabled=true
   )
  ]
  rendererClasses [&NiagaraSpriteRendererProperties]
  renderers [
   {rendererClass=&NiagaraSpriteRendererProperties}
   {rendererClass=&NiagaraSpriteRendererProperties, rendererIndex=1}
  ]
 }
 Spirals {
  bEnabled=false
  simTarget=GPUComputeSim
  emitterUpdateScript [
   &EmitterState()
   &SpawnRate(
    SpawnRate=&Multiply_Float(
     A=4
     B=&FloatFromCurve(
      FloatCurve=DataInterface() {propertyValues {Curve {keys [
          {interpMode=RCIM_Cubic, tangentMode=RCTM_User, time=0, value=1}
          {arriveTangent=-5.1692, interpMode=RCIM_Cubic, leaveTangent=-5.1692, tangentMode=RCTM_User, time=0.4, value=0}
         ]}}}
      CurveIndex=$Emitter.NormalizedLoopAge
     )
    )
    Use Spawn Probability=false
   )
  ]
  particleSpawnScript [
   /SetVariables(Particles.ScaleDetermination:NiagaraFloat=&MakeCustomFloatFromBool(
     Bool=&SetBoolByFloatComparison(A=&RandomRangeFloat(Minimum=0, Maximum=1), B=0.7)
     True Float=0
     False Float=1
    )) {isSetParameters=true}
   &InitializeParticle(
    Color=<rgba 1.8229 2.4049 50 15>
    Lifetime=&RandomRangeFloat(Minimum=2, Maximum=3)
    Mesh Scale Mode=::NewEnumerator3
    Mesh Uniform Scale=&Lerp_Float(A=0.93, B=0.8, Alpha=$Particles.ScaleDetermination)
    Sprite Size Mode=::NewEnumerator1
    Uniform Sprite Size Max=25
    UsePositionOffset=false
    Write Material Random=true
   )
   &InitialMeshOrientation(
    Mesh Axis To Orient=<0 0 1>
    Mesh Orientation Mode=::NewEnumerator2
    Orientation Vector=<0 0 1>
    Rotation=&RandomRangeVector(Minimum=<0 0 0>, Maximum=<0 0 1>)
    Use Rotation Vector=true
   )
   &ApplyInitialForces()
   &AddRotationalVelocity(Rotation Rate=&MakeVector(
     X=0
     Y=0
     Z=&FloatFromCurve(FloatCurve=DataInterface() {propertyValues {Curve {keys [<at 0 -2>,<at 0.5 -2>]}}}, CurveIndex=$Emitter.NormalizedLoopAge)
    ))
   &AddVelocity(Velocity=&MakeVector(
     X=0
     Y=0
     Z=&FloatFromCurve(
      FloatCurve=DataInterface() {propertyValues {Curve {keys [<at 0 0 RCIM_Cubic>, {arriveTangent=255.7545, interpMode=RCIM_Cubic, leaveTangent=255.7545, tangentMode=RCTM_User, time=1, value=120}]}}}
      CurveIndex=$Emitter.NormalizedLoopAge
      Scale Curve=$Particles.ScaleDetermination
     )
    ))
  ]
  particleUpdateScript [
   &ParticleState()
   &SolveForcesAndVelocity(Clamp Velocity=true, Limit Acceleration=false, Speed Limit=1200) {module=SolveForcesAndVelocity001}
   &SolveRotationalForcesAndVelocity()
   &SpriteRotationRate(Rotation Rate=&RandomRangeFloat(Minimum=-100, Maximum=100))
   &ScaleColor(
    Scale Alpha=&Multiply_Float(
     A=&FloatFromCurve(
      FloatCurve=DataInterface() {propertyValues {Curve {keys [
          {interpMode=RCIM_Cubic, tangentMode=RCTM_User, time=0, value=0}
          {interpMode=RCIM_Cubic, tangentMode=RCTM_User, time=0.1, value=1}
          {tangentMode=RCTM_User, time=1, value=0}
         ]}}}
      Scale Curve=10
     )
     B=&FN_LOD_Distance_Remap(Start Distance=2500, End Distance=5000, Minimum Output=0.05, Maximum Output=0.1)
    )
    Scale RGB=<0.05 0.05 0.05>
    ScaleA=true
    ScaleRGB=true
   )
   &ScaleMeshSize(Scale Factor=&VectorFromCurve(VectorCurve=DataInterface() {propertyValues {
       XCurve {keys [{interpMode=RCIM_Cubic, tangentMode=RCTM_User, time=0, value=1}, {interpMode=RCIM_Cubic, tangentMode=RCTM_User, time=1, value=0.95}]}
       YCurve {keys [{interpMode=RCIM_Cubic, tangentMode=RCTM_User, time=0, value=1}, {interpMode=RCIM_Cubic, tangentMode=RCTM_User, time=1, value=0.95}]}
       ZCurve {keys [{arriveTangent=0.9, interpMode=RCIM_Cubic, time=0, value=0.1}, {interpMode=RCIM_Cubic, leaveTangent=0.9, time=1, value=1.5}]}
      }}))
   &DynamicMaterialParameters(
    Index 0 Param 1=&RandomRangeFloat(Minimum=0, Maximum=1)
    Index 0 Param 2=2
    Index 0 Param 3=1
    Index 1 Param 1=50
    Index 1 Param 2=50
    Index 1 Param 3=500
    Module.Param0WriteEnabled=true
    Module.Param0WriteEnabled1=true
    Module.Param1WriteEnabled=true
    Module.Param1WriteEnabled1=true
    Module.Param2WriteEnabled=true
    Module.Param2WriteEnabled1=true
    Module.Param3WriteEnabled=true
    Module.Param3WriteEnabled1=true
    Write Parameter Index 1=true
   ) {module=DynamicMaterialParameters001}
  ]
  rendererClasses [&NiagaraMeshRendererProperties]
  renderers [{rendererClass=&NiagaraMeshRendererProperties, rendererIndex=0}]
 }
 Stone {
  bEnabled=false
  emitterUpdateScript [
   &EmitterState()
   &SpawnRate(SpawnRate=7, Use Spawn Probability=false)
  ]
  particleSpawnScript [
   &InitializeParticle(
    Color=<rgba 0.1198 0.099 0.083 1>
    Sprite Rotation Mode=::NewEnumerator2
    Sprite Size Mode=::NewEnumerator1
    Uniform Sprite Size Max=30
    Uniform Sprite Size Min=100
    UsePositionOffset=false
    Write Material Random=true
   )
   &VortexForce(Vortex Axis=<0 0 -1>, Vortex Force Amount=500) {enabled=false}
   &ApplyInitialForces(Apply Force to Position=true, Positional Force Warmup Time=-0.1)
  ]
  particleUpdateScript [
   &PointForce(Force Strength=-300, Module.Use Falloff=false, Module.UseLinearForce=false)
   &ParticleState()
   &ScaleSpriteSize(Uniform Curve Sprite Scale=DataInterface() {propertyValues {Curve {keys [
        {interpMode=RCIM_Cubic, tangentMode=RCTM_User, time=0, value=0}
        {interpMode=RCIM_Cubic, tangentMode=RCTM_User, time=0.5, value=1}
        {tangentMode=RCTM_User, time=1, value=0}
       ]}}})
   &SolveForcesAndVelocity(Clamp Velocity=false, Limit Acceleration=false)
   &ScaleColor(
    Scale Alpha=&FloatFromCurve(
     FloatCurve=DataInterface() {propertyValues {Curve {keys [{arriveTangent=0.0555, interpMode=RCIM_Cubic, leaveTangent=0.0555, tangentMode=RCTM_User, time=0, value=1}, <at 1 1>]}}}
     CurveIndex=$Emitter.NormalizedLoopAge
    )
    ScaleA=true
    ScaleRGB=true
   )
   &SpriteRotationRate(Rotation Rate=&RandomRangeFloat(Minimum=-100, Maximum=100, Evaluation Type=::NewEnumerator1)) {module=SpriteRotationRate001}
   &RotateAroundPoint(
    Radius=40
    Rotation Phase:NiagaraFloat=&RandomRangeFloat(Minimum=0, Maximum=1)
    Rotation Rate:NiagaraFloat=&RandomRangeFloat(Minimum=-1, Maximum=-0.5)
   )
  ]
  rendererClasses [&NiagaraSpriteRendererProperties]
  renderers [{rendererClass=&NiagaraSpriteRendererProperties}]
 }
 Trail_ring {
  bEnabled=false
  emitterUpdateScript [
   &EmitterState(Life Cycle Mode=::NewEnumerator1, Loop Behavior=::NewEnumerator1, Loop Delay=0.7, Loop Duration=2.8, UseLoopDelay=true)
   &SpawnRate(
    SpawnRate=&FloatFromCurve(FloatCurve=DataInterface() {propertyValues {Curve {keys [<at 0 0.3>,<at 0.7 1>]}}}, CurveIndex=$Emitter.NormalizedLoopAge, Scale Curve=6)
    Use Spawn Probability=false
   )
   /SetVariables() {isSetParameters=true}
  ]
  particleSpawnScript [
   &InitializeParticle(
    Color=<rgba 0.3333 0.4928 2 1>
    Lifetime Max=3
    Lifetime Mode=::NewEnumerator1
    Mesh Scale=&Multiply_VectorByFloat(
     Vector=&Multiply_VectorByFloat(
      Vector=<1 1 1.3>
      Float=&FloatFromCurve(
       FloatCurve=DataInterface() {propertyValues {Curve {keys [
           {arriveTangent=-0.0965, interpMode=RCIM_Cubic, leaveTangent=-0.0965, tangentMode=RCTM_User, time=0, value=1}
           {arriveTangent=-1.273, tangentMode=RCTM_User, time=0.9, value=0.7}
          ]}}}
       CurveIndex=$Emitter.NormalizedLoopAge
      )
     )
     Float=&RandomRangeFloat(Minimum=0.5, Maximum=1)
    )
    Mesh Scale Mode=::NewEnumerator4
    Position Offset=&MakeVector(
     X=0
     Y=0
     Z=&FloatFromCurve(
      FloatCurve=DataInterface() {propertyValues {Curve {keys [<at 0 0.2 RCIM_Cubic>,<at 1 1>]}}}
      CurveIndex=$Emitter.NormalizedLoopAge
      Scale Curve=150
     )
    )
    UsePositionOffset=true
    Write Material Random=true
   )
   &InitialMeshOrientation(Mesh Orientation Mode=::NewEnumerator2, Rotation=<0.25 0 0>, Use Rotation Vector=true)
   &InitialMeshOrientation(
    Mesh Orientation Mode=::NewEnumerator2
    Rotation=&RandomRangeVector(Minimum=<0 0 0>, Maximum=<0 0 0.5>)
    Use Rotation Vector=true
   ) {module=InitialMeshOrientation001}
   &InitialMeshOrientation(Mesh Orientation Mode=::NewEnumerator2, Rotation=<0 0.05 0>, Use Rotation Vector=true) {module=InitialMeshOrientation002}
   &DynamicMaterialParameters(
    Index 0 Param 1=&Multiply_Float(
     A=&RandomRangeFloat(Minimum=20, Maximum=30)
     B=&FloatFromCurve(
      FloatCurve=DataInterface() {propertyValues {Curve {keys [
          <at 1 2.5 RCIM_Cubic>
          {arriveTangent=-1.25, interpMode=RCIM_Cubic, leaveTangent=-1.25, time=1.6444, value=1.8}
          {arriveTangent=-0.0085, tangentMode=RCTM_User, time=2.2, value=1}
         ]}}}
      CurveIndex=$Emitter.Age
      Scale Curve=1.3
     )
    )
    Index 0 Param 2=&Multiply_Float(
     A=&RandomRangeFloat(Minimum=20, Maximum=30)
     B=&FloatFromCurve(
      FloatCurve=DataInterface() {propertyValues {Curve {keys [
          <at 1 2.5 RCIM_Cubic>
          {arriveTangent=-1.2, interpMode=RCIM_Cubic, leaveTangent=-1.2, time=1.6444, value=1.8}
          {arriveTangent=-0.0085, tangentMode=RCTM_User, time=2.25, value=1}
         ]}}}
      CurveIndex=$Emitter.Age
      Scale Curve=1.3
     )
    )
    Index 0 Param 3=1.5
    Index 0 Param 4=&RandomRangeFloat(Minimum=0, Maximum=1)
    Index 1 Param 1=&RandomRangeFloat(Minimum=1, Maximum=1)
    Index 1 Param 2=&FloatFromCurve(FloatCurve=DataInterface() {propertyValues {Curve {keys [{tangentMode=RCTM_User, time=0, value=1}, {tangentMode=RCTM_User, time=0.5, value=1}]}}})
    Module.Index0Param0Write=true
    Module.Index0Param1Write=true
    Module.Index0Param2Write=true
    Module.Index0Param3Write=true
    Module.Index1Param0Write=true
    Module.Index1Param1Write=true
    Module.Index1Param2Write=false
    Module.Index1Param3Write=false
    Write Parameter Index 1=true
   )
   &CameraOffset()
  ]
  particleUpdateScript [
   &ParticleState()
   &UpdateMeshOrientation(
    Rotation Coordinate Space=::NewEnumerator2
    Rotation Rate=&Multiply_Float(
     A=&RandomRangeFloat(Minimum=-0.05, Maximum=0.3)
     B=&FloatFromCurve(
      FloatCurve=DataInterface() {propertyValues {Curve {keys [
          {arriveTangent=-0.4674, interpMode=RCIM_Cubic, leaveTangent=-0.4674, tangentMode=RCTM_User, time=0.7, value=0.2}
          {arriveTangent=0.5986, interpMode=RCIM_Cubic, leaveTangent=0.5986, tangentMode=RCTM_User, time=1.9, value=0.5}
          {arriveTangent=-0.9565, interpMode=RCIM_Cubic, leaveTangent=-0.9565, tangentMode=RCTM_User, time=2.5, value=5}
         ]}}}
      CurveIndex=$Emitter.Age
     )
    )
    Rotation Vector=<0 0 1>
   ) {module=UpdateMeshOrientation001}
   &ScaleMeshSize(Scale Factor=&Multiply_VectorByFloat(
     Vector=<1 1.2 1>
     Float=&Multiply_Float(
      A=0.7
      B=&FloatFromCurve(FloatCurve=DataInterface() {propertyValues {Curve {keys [<at 0 0.8>,<at 1 1.2>]}}})
     )
    ))
   &ScaleColor(
    Scale Alpha=&FloatFromCurve(
     FloatCurve=DataInterface() {propertyValues {Curve {keys [
         {interpMode=RCIM_Cubic, tangentMode=RCTM_User, time=0, value=0}
         <at 0.2 1>
         {interpMode=RCIM_Cubic, tangentMode=RCTM_User, time=0.7, value=1}
         {tangentMode=RCTM_User, time=1, value=0}
        ]}}}
     Scale Curve=&FloatFromCurve(FloatCurve=DataInterface() {propertyValues {Curve {keys [<at 0.7 0>,<at 1.2 1>,<at 3.8 1>,<at 4 0>]}}}, CurveIndex=$Emitter.Age)
    )
    Scale RGB=&Multiply_VectorByFloat(Vector=<1 1 1>, Float=1)
    ScaleA=true
    ScaleRGB=true
   )
  ]
  rendererClasses [&NiagaraMeshRendererProperties]
  renderers [{rendererClass=&NiagaraMeshRendererProperties, rendererIndex=0}]
 }
 Trail_ring001 {
  bEnabled=false
  emitterUpdateScript [
   &EmitterState(Life Cycle Mode=::NewEnumerator1, Loop Behavior=::NewEnumerator1, Loop Delay=0.7, Loop Duration=2.2, UseLoopDelay=true)
   &SpawnRate(SpawnRate=40, Use Spawn Probability=false)
  ]
  particleSpawnScript [
   &InitializeParticle(
    Color=&MakeLinearColorFromVectorAndFloat(
     "Vector (RGB)"=&MakeVectorFromLinearColorRGB(Color=&LerpLinearColors(
       StartColor=<rgba 0.005 0.005 0.005 1>
       EndColor=&MakeCustomLinearColorFromBool(Bool=$User.bIsFrontEndPreview, True Color=<rgba 0.015 0 0.03 0.7>, False Color=<rgba 0.025 0 0.05 1>)
       LerpFactor=$User.VariantSwitch
      ))
     "Float (Alpha)"=&FloatFromCurve(
      FloatCurve=DataInterface() {propertyValues {Curve {keys [
          <at 0 1>
          {arriveTangent=-1.4286, interpMode=RCIM_Cubic, leaveTangent=-1.4286, time=0.5, value=1}
          {arriveTangent=-0.3079, interpMode=RCIM_Cubic, leaveTangent=-0.3079, tangentMode=RCTM_User, time=0.7, value=0}
         ]}}}
      CurveIndex=$Emitter.NormalizedLoopAge
     )
    )
    Lifetime Max=3
    Lifetime Min=2
    Lifetime Mode=::NewEnumerator1
    Mesh Scale=&Multiply_VectorByFloat(
     Vector=&Multiply_VectorByFloat(
      Vector=<1 1 1>
      Float=&FloatFromCurve(
       FloatCurve=DataInterface() {propertyValues {Curve {keys [
           {arriveTangent=-0.0965, interpMode=RCIM_Cubic, leaveTangent=-0.0965, tangentMode=RCTM_User, time=0, value=1.1}
           {arriveTangent=-0.4786, interpMode=RCIM_Cubic, leaveTangent=-0.4786, time=0.4, value=0.7}
           <at 0.6268 0.8>
           <at 0.7675 0.5>
           {arriveTangent=-0.2846, interpMode=RCIM_Cubic, leaveTangent=-0.2846, tangentMode=RCTM_User, time=0.9, value=0.8}
          ]}}}
       CurveIndex=$Emitter.NormalizedLoopAge
       Scale Curve=0.7
      )
     )
     Float=&RandomRangeFloat(Minimum=0.6, Maximum=1.5)
    )
    Mesh Scale Mode=::NewEnumerator4
    Position Offset=&MakeVector(
     X=0
     Y=0
     Z=&Add_Float(
      A=&FloatFromCurve(
       FloatCurve=DataInterface() {propertyValues {Curve {keys [<at 0 0 RCIM_Cubic>,<at 0.7 1>]}}}
       CurveIndex=$Emitter.NormalizedLoopAge
       Scale Curve=&FN_PickFloatDependOnPlatform(TargetPlatformActive=150, TargetPlatformNoActive=130)
      )
      B=&FN_PickFloatDependOnPlatform(TargetPlatformActive=0, TargetPlatformNoActive=35)
     )
    )
    UsePositionOffset=true
    Write Material Random=true
   )
   &InitialMeshOrientation(Mesh Orientation Mode=::NewEnumerator2, Rotation=<0.25 0 0>, Use Rotation Vector=true)
   &InitialMeshOrientation(
    Mesh Orientation Mode=::NewEnumerator2
    Rotation=&RandomRangeVector(Minimum=<0 0 0>, Maximum=<0 0 1>)
    Use Rotation Vector=true
   ) {module=InitialMeshOrientation001}
   &InitialMeshOrientation(Mesh Orientation Mode=::NewEnumerator2, Rotation=<0 0.05 0>, Use Rotation Vector=true) {module=InitialMeshOrientation002}
   &DynamicMaterialParameters(
    Index 0 Param 1=&Multiply_Float(
     A=&RandomRangeFloat(Minimum=30, Maximum=30)
     B=&FloatFromCurve(
      FloatCurve=DataInterface() {propertyValues {Curve {keys [
          <at 1 2.5 RCIM_Cubic>
          {arriveTangent=-1.44, interpMode=RCIM_Cubic, leaveTangent=-1.44, time=1.6444, value=1.8}
          {arriveTangent=-0.0085, tangentMode=RCTM_User, time=2.25, value=0.7}
         ]}}}
      CurveIndex=$Emitter.Age
      Scale Curve=0
     )
    )
    Index 0 Param 2=&Multiply_Float(
     A=&RandomRangeFloat(Minimum=20, Maximum=30)
     B=&FloatFromCurve(
      FloatCurve=DataInterface() {propertyValues {Curve {keys [
          <at 1 2.5 RCIM_Cubic>
          {arriveTangent=-1.44, interpMode=RCIM_Cubic, leaveTangent=-1.44, time=1.6444, value=1.8}
          {arriveTangent=-0.0085, tangentMode=RCTM_User, time=2.25, value=0.7}
         ]}}}
      CurveIndex=$Emitter.Age
      Scale Curve=0
     )
    )
    Index 0 Param 3=1.5
    Index 0 Param 4=&RandomRangeFloat(Minimum=0, Maximum=1)
    Index 1 Param 1=&RandomRangeFloat(Minimum=0.2, Maximum=1)
    Index 1 Param 2=&FloatFromCurve(FloatCurve=DataInterface() {propertyValues {Curve {keys [{tangentMode=RCTM_User, time=0, value=1}, {tangentMode=RCTM_User, time=0.5, value=1}]}}})
    Module.Index0Param0Write=true
    Module.Index0Param1Write=true
    Module.Index0Param2Write=true
    Module.Index0Param3Write=true
    Module.Index1Param0Write=true
    Module.Index1Param1Write=true
    Module.Index1Param2Write=false
    Module.Index1Param3Write=false
    Write Parameter Index 1=true
   )
   &CameraOffset()
  ]
  particleUpdateScript [
   &ParticleState()
   &UpdateMeshOrientation(
    Rotation Coordinate Space=::NewEnumerator2
    Rotation Rate=&Multiply_Float(
     A=&RandomRangeFloat(Minimum=-0.1, Maximum=0.3)
     B=&FloatFromCurve(
      FloatCurve=DataInterface() {propertyValues {Curve {keys [
          {arriveTangent=-0.4674, interpMode=RCIM_Cubic, leaveTangent=-0.4674, tangentMode=RCTM_User, time=0.7, value=0.5}
          {arriveTangent=0.5986, interpMode=RCIM_Cubic, leaveTangent=0.5986, tangentMode=RCTM_User, time=1.9, value=1}
          {arriveTangent=-0.9565, interpMode=RCIM_Cubic, leaveTangent=-0.9565, tangentMode=RCTM_User, time=2.5, value=4}
         ]}}}
      CurveIndex=$Emitter.Age
     )
    )
    Rotation Vector=<0 0 1>
   )
   &ScaleMeshSize(Scale Factor=&Multiply_VectorByFloat(
     Vector=<1 1 1>
     Float=&Multiply_Float(
      A=0.8
      B=&FloatFromCurve(FloatCurve=DataInterface() {propertyValues {Curve {keys [<at 0 0.8>,<at 1 1.2>]}}})
     )
    ))
   &ScaleColor(
    Scale Alpha=&FloatFromCurve(
     FloatCurve=DataInterface() {propertyValues {Curve {keys [
         {interpMode=RCIM_Cubic, tangentMode=RCTM_User, time=0, value=0}
         {interpMode=RCIM_Cubic, tangentMode=RCTM_User, time=0.5, value=1}
         {tangentMode=RCTM_User, time=1, value=0}
        ]}}}
     Scale Curve=&FloatFromCurve(FloatCurve=DataInterface() {propertyValues {Curve {keys [<at 0.7 1>,<at 3.2 1>,<at 3.5 0>]}}}, CurveIndex=$Emitter.Age)
    )
    Scale RGB=&Multiply_VectorByFloat(Vector=<1 1 1>, Float=1)
    ScaleA=true
    ScaleRGB=true
   )
  ]
  rendererClasses [&NiagaraMeshRendererProperties]
  renderers [{rendererClass=&NiagaraMeshRendererProperties, rendererIndex=0}]
 }
 Trail_ring002 {
  bEnabled=false
  emitterUpdateScript [
   &EmitterState(Life Cycle Mode=::NewEnumerator1, Loop Behavior=::NewEnumerator1, Loop Delay=0.7, Loop Duration=2.2, UseLoopDelay=true)
   &SpawnRate(SpawnRate=15, Use Spawn Probability=false)
  ]
  particleSpawnScript [
   &InitializeParticle(
    Color=&MakeLinearColorFromVectorAndFloat(
     "Vector (RGB)"=&MakeVectorFromLinearColorRGB(Color=&LerpLinearColors(StartColor=<rgba 0 8 1.948 2>, EndColor=<rgba 6.4 1.6 16 2>, LerpFactor=$User.VariantSwitch))
     "Float (Alpha)"=&FloatFromCurve(FloatCurve=DataInterface() {propertyValues {Curve {keys [<at 0 1>,<at 0.5 1>,<at 0.6 0>]}}}, CurveIndex=$Emitter.NormalizedLoopAge)
    )
    Lifetime Min=1.5
    Lifetime Mode=::NewEnumerator1
    Mesh Scale=&Multiply_VectorByFloat(
     Vector=&Multiply_VectorByFloat(
      Vector=<1 1 1.3>
      Float=&FloatFromCurve(
       FloatCurve=DataInterface() {propertyValues {Curve {keys [
           {arriveTangent=-0.0965, interpMode=RCIM_Cubic, leaveTangent=-0.0965, tangentMode=RCTM_User, time=0, value=1}
           {arriveTangent=-0.3191, interpMode=RCIM_Cubic, leaveTangent=-0.3191, time=0.4, value=0.5}
           <at 0.6268 0.8>
           <at 0.7675 0.5>
           {arriveTangent=-0.2846, interpMode=RCIM_Cubic, leaveTangent=-0.2846, tangentMode=RCTM_User, time=0.9, value=0.8}
          ]}}}
       CurveIndex=$Emitter.NormalizedLoopAge
       Scale Curve=0.8
      )
     )
     Float=&RandomRangeFloat(Minimum=0.6, Maximum=1)
    )
    Mesh Scale Mode=::NewEnumerator4
    Position Offset=&MakeVector(
     X=0
     Y=0
     Z=&Add_Float(
      A=&FloatFromCurve(
       FloatCurve=DataInterface() {propertyValues {Curve {keys [<at 0 0 RCIM_Cubic>,<at 0.8 1>]}}}
       CurveIndex=$Emitter.NormalizedLoopAge
       Scale Curve=&FN_PickFloatDependOnPlatform(TargetPlatformActive=150, TargetPlatformNoActive=130)
      )
      B=&FN_PickFloatDependOnPlatform(TargetPlatformActive=0, TargetPlatformNoActive=35)
     )
    )
    UsePositionOffset=true
    Write Material Random=true
   )
   &InitialMeshOrientation(Mesh Orientation Mode=::NewEnumerator2, Rotation=<0.25 0 0>, Use Rotation Vector=true)
   &InitialMeshOrientation(
    Mesh Orientation Mode=::NewEnumerator2
    Rotation=&RandomRangeVector(Minimum=<0 0 0>, Maximum=<0 0 1>)
    Use Rotation Vector=true
   ) {module=InitialMeshOrientation001}
   &InitialMeshOrientation(Mesh Orientation Mode=::NewEnumerator2, Rotation=<0 0.05 0>, Use Rotation Vector=true) {module=InitialMeshOrientation002}
   &DynamicMaterialParameters(
    Index 0 Param 1=&Multiply_Float(
     A=&RandomRangeFloat(Minimum=20, Maximum=30)
     B=&FloatFromCurve(
      FloatCurve=DataInterface() {propertyValues {Curve {keys [
          <at 1 2.5 RCIM_Cubic>
          {arriveTangent=-1.25, interpMode=RCIM_Cubic, leaveTangent=-1.25, time=1.6444, value=1.8}
          {arriveTangent=-0.0085, tangentMode=RCTM_User, time=2.2, value=1}
         ]}}}
      CurveIndex=$Emitter.Age
      Scale Curve=0
     )
    )
    Index 0 Param 2=&Multiply_Float(
     A=&RandomRangeFloat(Minimum=20, Maximum=30)
     B=&FloatFromCurve(
      FloatCurve=DataInterface() {propertyValues {Curve {keys [
          <at 1 2.5 RCIM_Cubic>
          {arriveTangent=-1.2, interpMode=RCIM_Cubic, leaveTangent=-1.2, time=1.6444, value=1.8}
          {arriveTangent=-0.0085, tangentMode=RCTM_User, time=2.25, value=1}
         ]}}}
      CurveIndex=$Emitter.Age
      Scale Curve=0
     )
    )
    Index 0 Param 3=1.5
    Index 0 Param 4=&RandomRangeFloat(Minimum=0, Maximum=1)
    Index 1 Param 1=&RandomRangeFloat(Minimum=1, Maximum=1)
    Index 1 Param 2=&FloatFromCurve(FloatCurve=DataInterface() {propertyValues {Curve {keys [{tangentMode=RCTM_User, time=0, value=1}, {tangentMode=RCTM_User, time=0.5, value=1}]}}})
    Module.Index0Param0Write=true
    Module.Index0Param1Write=true
    Module.Index0Param2Write=true
    Module.Index0Param3Write=true
    Module.Index1Param0Write=true
    Module.Index1Param1Write=true
    Module.Index1Param2Write=false
    Module.Index1Param3Write=false
    Write Parameter Index 1=true
   )
   &CameraOffset()
  ]
  particleUpdateScript [
   &ParticleState()
   &UpdateMeshOrientation(
    Rotation Coordinate Space=::NewEnumerator2
    Rotation Rate=&Multiply_Float(
     A=&RandomRangeFloat(Minimum=0.1, Maximum=0.4)
     B=&FloatFromCurve(
      FloatCurve=DataInterface() {propertyValues {Curve {keys [
          {arriveTangent=-0.4674, interpMode=RCIM_Cubic, leaveTangent=-0.4674, tangentMode=RCTM_User, time=0.7, value=0.5}
          {arriveTangent=0.5986, interpMode=RCIM_Cubic, leaveTangent=0.5986, tangentMode=RCTM_User, time=1.9, value=1}
          {arriveTangent=-0.9565, interpMode=RCIM_Cubic, leaveTangent=-0.9565, tangentMode=RCTM_User, time=2.5, value=4}
         ]}}}
      CurveIndex=$Emitter.Age
     )
    )
    Rotation Vector=<0 0 1>
   ) {module=UpdateMeshOrientation001}
   &ScaleMeshSize(Scale Factor=&Multiply_VectorByFloat(
     Vector=<1 2 1>
     Float=&Multiply_Float(
      A=1
      B=&FloatFromCurve(FloatCurve=DataInterface() {propertyValues {Curve {keys [<at 0 0.8>,<at 1 1.2>]}}})
     )
    ))
   &ScaleColor(
    Scale Alpha=&FloatFromCurve(
     FloatCurve=DataInterface() {propertyValues {Curve {keys [
         {interpMode=RCIM_Cubic, tangentMode=RCTM_User, time=0, value=0}
         <at 0.2 1>
         {interpMode=RCIM_Cubic, tangentMode=RCTM_User, time=0.7, value=1}
         {tangentMode=RCTM_User, time=1, value=0}
        ]}}}
     Scale Curve=&FloatFromCurve(FloatCurve=DataInterface() {propertyValues {Curve {keys [<at 0.7 0>,<at 1 1>,<at 3.2 1>,<at 3.5 0>]}}}, CurveIndex=$Emitter.Age)
    )
    Scale RGB=&Multiply_VectorByFloat(Vector=<1 1 1>, Float=1)
    ScaleA=true
    ScaleRGB=true
   )
  ]
  rendererClasses [&NiagaraMeshRendererProperties]
  renderers [{rendererClass=&NiagaraMeshRendererProperties, rendererIndex=0}]
 }
 Trail_ring003 {
  emitterUpdateScript [
   &EmitterState(Life Cycle Mode=::NewEnumerator1, Loop Behavior=::NewEnumerator1, Loop Duration=2.7, UseLoopDelay=true)
   &SpawnRate(
    SpawnRate=&FloatFromCurve(FloatCurve=DataInterface() {propertyValues {Curve {keys [<at 0 4 RCIM_Cubic>,<at 1 12 RCIM_Cubic>]}}}, CurveIndex=$Emitter.NormalizedLoopAge)
    Use Spawn Probability=false
   )
   /SetVariables(Emitter.Speed:NiagaraFloat=&RandomRangeFloat(Minimum=-1.2, Maximum=-2)) {isSetParameters=true}
  ]
  particleSpawnScript [
   &InitializeParticle(
    Color=&FNTimeOfDayColor(
     Night Color=<rgba 0.0938 0.1153 1 0.5>
     Morning Color=<rgba 0.1562 0.1763 1 0.4>
     Day Color=<rgba 0.1562 0.1763 1 0.5>
     Evening Color=<rgba 0.2083 0.2272 1 0.4>
    )
    Lifetime Max=2.4
    Lifetime Mode=::NewEnumerator1
    Mesh Scale=&Multiply_VectorByFloat(
     Vector=&Multiply_VectorByFloat(
      Vector=&RandomRangeVector(Minimum=<0.9 0.9 0.8>, Maximum=<1.1 1.1 1>)
      Float=&FloatFromCurve(
       FloatCurve=DataInterface() {propertyValues {Curve {keys [
           {arriveTangent=-0.0965, interpMode=RCIM_Cubic, leaveTangent=-0.0965, tangentMode=RCTM_User, time=0, value=0.8}
           {arriveTangent=1.2889, tangentMode=RCTM_User, time=1, value=1.5}
          ]}}}
       CurveIndex=$Emitter.NormalizedLoopAge
       Scale Curve=0.8
      )
     )
     Float=&RandomRangeFloat(Minimum=0.9, Maximum=1)
    )
    Mesh Scale Mode=::NewEnumerator4
    Position Offset=&MakeVector(
     X=0
     Y=0
     Z=&FloatFromCurve(
      FloatCurve=DataInterface() {propertyValues {Curve {keys [<at 0 0 RCIM_Cubic>, {arriveTangent=2.944, tangentMode=RCTM_User, time=1, value=1}]}}}
      CurveIndex=$Emitter.NormalizedLoopAge
      Scale Curve=150
     )
    )
    UsePositionOffset=true
    Write Material Random=true
   )
   &InitialMeshOrientation(Mesh Orientation Mode=::NewEnumerator2, Rotation=<0.25 0 0>, Use Rotation Vector=true)
   &InitialMeshOrientation(
    Mesh Orientation Mode=::NewEnumerator2
    Rotation=&RandomRangeVector(Minimum=<0 0 0>, Maximum=<0 0 1>)
    Use Rotation Vector=true
   ) {module=InitialMeshOrientation001}
   &InitialMeshOrientation(
    Mesh Orientation Mode=::NewEnumerator2
    Rotation=&RandomRangeVector(Minimum=<0 -0.1 0>, Maximum=<0 -0.03 0>)
    Use Rotation Vector=true
   ) {module=InitialMeshOrientation002}
   &DynamicMaterialParameters(
    Index 0 Param 1=&Multiply_Float(
     A=&RandomRangeFloat(Minimum=20, Maximum=30)
     B=&FloatFromCurve(
      FloatCurve=DataInterface() {propertyValues {Curve {keys [
          <at 1 2.5 RCIM_Cubic>
          {arriveTangent=-1.25, interpMode=RCIM_Cubic, leaveTangent=-1.25, time=1.6444, value=1.8}
          {arriveTangent=-0.0085, tangentMode=RCTM_User, time=2.2, value=1}
         ]}}}
      CurveIndex=$Emitter.Age
      Scale Curve=1.3
     )
    )
    Index 0 Param 2=&Multiply_Float(
     A=&RandomRangeFloat(Minimum=20, Maximum=30)
     B=&FloatFromCurve(
      FloatCurve=DataInterface() {propertyValues {Curve {keys [
          <at 1 2.5 RCIM_Cubic>
          {arriveTangent=-1.2, interpMode=RCIM_Cubic, leaveTangent=-1.2, time=1.6444, value=1.8}
          {arriveTangent=-0.0085, tangentMode=RCTM_User, time=2.25, value=1}
         ]}}}
      CurveIndex=$Emitter.Age
      Scale Curve=1.3
     )
    )
    Index 0 Param 3=1.5
    Index 0 Param 4=&RandomRangeFloat(Minimum=0, Maximum=1)
    Index 1 Param 1=&RandomRangeFloat(Minimum=1, Maximum=1)
    Index 1 Param 2=&FloatFromCurve(FloatCurve=DataInterface() {propertyValues {Curve {keys [{tangentMode=RCTM_User, time=0, value=1}, {tangentMode=RCTM_User, time=0.5, value=1}]}}})
    Module.Index0Param0Write=true
    Module.Index0Param1Write=true
    Module.Index0Param2Write=true
    Module.Index0Param3Write=true
    Module.Index1Param0Write=true
    Module.Index1Param1Write=true
    Module.Index1Param2Write=false
    Module.Index1Param3Write=false
    Write Parameter Index 1=true
   )
   &CameraOffset()
  ]
  particleUpdateScript [
   &ParticleState()
   &UpdateMeshOrientation(
    Rotation Coordinate Space=::NewEnumerator2
    Rotation Rate=&Multiply_Float(
     A=&RandomRangeFloat(Minimum=0, Maximum=0.3)
     B=&FloatFromCurve(
      FloatCurve=DataInterface() {propertyValues {Curve {keys [
          {arriveTangent=0.1006, interpMode=RCIM_Cubic, leaveTangent=0.1006, tangentMode=RCTM_User, time=0.7, value=0.2}
          {arriveTangent=0.5986, interpMode=RCIM_Cubic, leaveTangent=0.5986, tangentMode=RCTM_User, time=1.9, value=0.3}
          {arriveTangent=0.5342, interpMode=RCIM_Cubic, leaveTangent=0.5342, tangentMode=RCTM_User, time=2.5, value=1}
         ]}}}
      CurveIndex=$Emitter.Age
      Scale Curve=-1
     )
    )
    Rotation Vector=<0 0 1>
   ) {module=UpdateMeshOrientation001}
   &ScaleMeshSize(Scale Factor=&Multiply_VectorByFloat(
     Vector=<1 1.2 1>
     Float=&Multiply_Float(
      A=0.7
      B=&FloatFromCurve(FloatCurve=DataInterface() {propertyValues {Curve {keys [<at 0 0.8>,<at 1 1.2>]}}})
     )
    ))
   &ScaleColor(
    Scale Alpha=&FloatFromCurve(
     FloatCurve=DataInterface() {propertyValues {Curve {keys [
         {interpMode=RCIM_Cubic, tangentMode=RCTM_User, time=0, value=0}
         <at 0.2 1>
         {interpMode=RCIM_Cubic, tangentMode=RCTM_User, time=0.7, value=1}
         {tangentMode=RCTM_User, time=1, value=0}
        ]}}}
     Scale Curve=&FloatFromCurve(FloatCurve=DataInterface() {propertyValues {Curve {keys [<at 0.7 0>,<at 1.2 1>,<at 4 1>,<at 4.1 0>]}}}, CurveIndex=$Emitter.Age)
    )
    Scale RGB=&Multiply_VectorByFloat(Vector=<1 1 1>, Float=1)
    ScaleA=true
    ScaleRGB=true
   )
  ]
  rendererClasses [&NiagaraMeshRendererProperties]
  renderers [{rendererClass=&NiagaraMeshRendererProperties, rendererIndex=0}]
 }
 Wind_Wave {
  emitterSpawnScript [/SetVariables(
    Emitter.DarkColor:LinearColor=&FNTimeOfDayColor(
     Night Color=<rgba 0.036 0.0808 0.4062 1>
     Morning Color=<rgba 0.1164 0.1515 0.4062 1>
     Day Color=<rgba 0.1164 0.1515 0.4062 1>
     Evening Color=<rgba 0.1164 0.1515 0.4062 1>
    )
    Emitter.MainColor:LinearColor=&SelectLinearColorFromArray(
     Array Sampling Mode=::NewEnumerator0
     Direct Array Index=&RandomRangeInt(Minimum=0, Maximum=1)
     Color Selection Array=DataInterface() {propertyValues {ColorData [<rgba 0.0487 0.1098 1 1>, <rgba 0.0819 0.1782 1 1>]}}
     Random Seed=0
    )
   ) {isSetParameters=true}]
  emitterUpdateScript [
   &EmitterState()
   &SpawnBurst_Instantaneous(Use Spawn Probability=false, UseLoopCountLimit=true)
  ]
  particleSpawnScript [
   &PartitionParticles()
   &InitializeParticle(
    Color=<rgba 1 1 1 8>
    Lifetime=4.7
    Mesh Scale Mode=::NewEnumerator3
    Mesh Uniform Scale=&Multiply_Float(A=1.35, B=$User.Scale)
    Position Offset=<0 0 5>
    UsePositionOffset=true
    Write Material Random=true
   )
   &InitialMeshOrientation(
    Mesh Orientation Mode=::NewEnumerator2
    Orientation Vector=<0 0 1>
    Rotation=&RandomRangeVector(Minimum=<-5 0 0>, Maximum=<5 0 0>)
    Use Rotation Vector=true
   )
   &CameraOffset(Camera Offset Amount=5)
  ]
  particleUpdateScript [
   &ParticleState(Kill Particles When Lifetime Has Elapsed=false)
   &ScaleMeshSize(Scale Factor=&VectorFromFloat(Value=&FloatFromCurve(
      FloatCurve=DataInterface() {propertyValues {Curve {keys [
          {arriveTangent=4.4607, interpMode=RCIM_Cubic, leaveTangent=4.4607, tangentMode=RCTM_User, time=0, value=0.35}
          {arriveTangent=0.8667, interpMode=RCIM_Cubic, leaveTangent=0.8667, time=0.1312, value=0.6818}
          {arriveTangent=1.1375, interpMode=RCIM_Cubic, leaveTangent=1.1375, tangentMode=RCTM_User, time=0.75, value=1}
         ]}}}
      Scale Curve=0.1
     ))) {enabled=false}
   &DynamicMaterialParameters(
    Index 0 Param 1=&FloatFromCurve(FloatCurve=DataInterface() {propertyValues {Curve {keys [
         {arriveTangent=9.0725, interpMode=RCIM_Cubic, leaveTangent=9.0725, tangentMode=RCTM_User, time=0, value=0}
         {arriveTangent=-2.5249, interpMode=RCIM_Cubic, leaveTangent=-2.5249, tangentMode=RCTM_User, time=0.1408, value=0.9847}
         {arriveTangent=-0.0405, interpMode=RCIM_Cubic, leaveTangent=-0.0405, tangentMode=RCTM_User, time=1, value=0}
        ]}}})
    Index 0 Param 2=&FloatFromCurve(FloatCurve=DataInterface() {propertyValues {Curve {keys [{arriveTangent=2.105, interpMode=RCIM_Cubic, leaveTangent=2.105, tangentMode=RCTM_User, time=0.1, value=0.1}, <at 1 1>]}}})
    Index 0 Param 3=100
    Index 0 Param 4=100
    Module.Param0WriteEnabled=false
    Module.Param1WriteEnabled=false
    Module.Param2WriteEnabled=true
    Module.Param3WriteEnabled=true
   ) {enabled=false}
   &ScaleColor(
    Scale Alpha=&FloatFromCurve(
     FloatCurve=DataInterface() {propertyValues {Curve {keys [
         {arriveTangent=8.3055, interpMode=RCIM_Cubic, leaveTangent=8.3055, tangentMode=RCTM_User, time=0.1, value=0}
         {interpMode=RCIM_Cubic, tangentMode=RCTM_User, time=0.5, value=1}
         {arriveTangent=-1.2034, tangentMode=RCTM_User, time=1, value=0}
        ]}}}
     Scale Curve=&Lerp_Float(A=0, B=1, Alpha=$System.Timeline.NormalizedTime)
    )
    ScaleA=true
    ScaleRGB=true
   ) {module=ScaleColor001}
   &KillParticles(Kill Particles=&Boolean_AndOperation(
     Bool1=&Boolean_NotOperation(Bool=/Game/Effects/Niagara/DynamicInputs/State/FN_IsExecutionStateActive())
     Bool2=&Boolean_NotOperation(Bool=$System.Timeline.IsPlaying)
    ))
  ]
  rendererClasses [&NiagaraMeshRendererProperties]
  renderers [{rendererClass=&NiagaraMeshRendererProperties, rendererIndex=0}]
 }
}
summary {
 systemName=MS_Emote_GarlicWish_Vortex
 emitters [
  {emitterName=NS_SoftAirflow, rendererClasses [&NiagaraMeshRendererProperties]}
  {emitterName=Wind_Wave, rendererClasses [&NiagaraMeshRendererProperties]}
  {bEnabled=false, emitterName=Ground_FX001, rendererClasses [&NiagaraMeshRendererProperties]}
  {emitterName=Ground_FX, rendererClasses [&NiagaraMeshRendererProperties]}
  {bEnabled=false, emitterName=Spirals, simTarget=GPUComputeSim, rendererClasses [&NiagaraMeshRendererProperties]}
  {bEnabled=false, emitterName=Debris_01002, simTarget=GPUComputeSim, rendererClasses [&NiagaraSpriteRendererProperties]}
  {bEnabled=false, emitterName=Stone, rendererClasses [&NiagaraSpriteRendererProperties]}
  {emitterName=Smoke, simTarget=GPUComputeSim, rendererClasses [&NiagaraSpriteRendererProperties]}
  {bEnabled=false, emitterName=Trail_ring, rendererClasses [&NiagaraMeshRendererProperties]}
  {emitterName=Trail_ring003, rendererClasses [&NiagaraMeshRendererProperties]}
  {bEnabled=false, emitterName=Trail_ring002, rendererClasses [&NiagaraMeshRendererProperties]}
  {bEnabled=false, emitterName=Trail_ring001, rendererClasses [&NiagaraMeshRendererProperties]}
 ]
 userVariables [
  User.bIsFrontEnd {type=NiagaraBool}
  User.bIsFrontEndPreview {type=NiagaraBool}
  User.RampupTimeDebris {type=NiagaraFloat, defaultValue=1, description=""}
  User.RampupTimeSmoke {type=NiagaraFloat, defaultValue=1, description=""}
  User.Scale {type=NiagaraFloat, defaultValue=1, description=""}
  User.VariantSwitch {type=NiagaraFloat, defaultValue=0, description=""}
  User.VortexHeight {type=Vector3f, description="", defaultValue {struct=Vector3f, value=<0 0 1>}}
 ]
}
