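# Niagara trail effect for the GarlicWhisk emote, drawing ribbons that follow two bones or sockets on a skeletal mesh.
# Two active emitters each spawn a ribbon pair, picking the bone by spawn group: one takes the system color, the other is faint, short-lived white.
# Both taper width and fade from a bright blue-tinted boost, with system variables driving spawn rate, alpha and a Dive-blended material parameter.
# Single-bone ribbon variants and mesh speed lines are disabled.
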NiagaraSystem /BRCosmetics/Animation/Game/MainPlayer/Emotes/GarlicWhisk/FX/NS_Emote_GarlicWhisk_Trail
dependencies {
 usedDataInterfaces [&NiagaraDataInterfaceCurve, &NiagaraDataInterfaceColorCurve, &NiagaraDataInterfaceArrayFloat]
 usedDynamicInputs []
 usedModules [
  &SystemState
  &self:SystemSpawnScript.SystemScriptSource.SystemScriptGraph.NiagaraNodeAssignment_0.SetVariables_#1
  &self:SystemSpawnScript.SystemScriptSource.SystemScriptGraph.NiagaraNodeAssignment_1.SetVariables_#2
  &self:SystemSpawnScript.SystemScriptSource.SystemScriptGraph.NiagaraNodeAssignment_3.SetVariables_#3
  &self:SystemSpawnScript.SystemScriptSource.SystemScriptGraph.NiagaraNodeAssignment_2.SetVariables_#4
 ]
}
emitters {
 RibbonWhite {
  bEnabled=false
  emitterUpdateScript [
   &EmitterState()
   /Game/Effects/Niagara/Modules/Skydiving/FN_DivingAttributes/FN_EmitterDivingAttributes_v1(Bone Index=1, SkeletalMesh:NiagaraDataInterfaceSkeletalMesh=$User.SkeletalMesh) {enabled=false}
   &FN_SpawnRibbon_v2(Spawn Mode=::NewEnumerator0, SpawnRate:NiagaraFloat=$System.SpawnRate)
  ]
  particleSpawnScript [
   &FN_SetSpawnGroup(Number of Spawn Groups=1)
   &FN_InitializeRibbon_v3(
    Color:LinearColor=$System.Linear Color
    Custom Offset=true
    Enable Skeletal Position=true
    Local Vector Offset:Vector3f=<0 1 0>
    Ribbon Width=25
   )
   &AddVelocity(
    Scale Added Velocity=&Lerp_Vector(A=<1 1 1>, B=<4 4 1>, Alpha=$System.Dive)
    Velocity=&FN_FrontEndVector3Modifier(Gameplay=<0 0 550>, Front End=<0 0 550>)
   ) {enabled=false}
   &AddVelocity(Velocity=&MakeVector(
     X=&SelectFloatFromArray(Direct Array Index=$Particles.SpawnGroup, Float Selection Array=DataInterface() {propertyValues {FloatData [100,-100]}})
     Y=0
     Z=0
    )) {enabled=false, module=AddVelocity001}
   &SkeletalMeshLocation(
    Apply Sampled Orientation to Mesh Particles=true
    Bone / Socket Index=1
    Bone Sampling Mode=::NewEnumerator4
    Inherit Velocity=false
    Interpolate Mesh Sampling=false
    MaskSpawnGroup=false
    Sampled Position Offset Coordinate Space:ENiagaraMeshTransforms=::NewEnumerator3
    Skeletal Mesh=$System.Skeletal Mesh
   )
   &DynamicMaterialParameters(
    Index 0 Param 1=&MaskFloatBySpawnGroup(Masked Float=1, Passthrough Float=0.8)
    Index 0 Param 2=&MaskFloatBySpawnGroup(Masked Float=0.4, Passthrough Float=3)
    Index 0 Param 3=1
    Index 1 Param 1=&Lerp_Float(A=5, B=10, Alpha=$System.Dive)
    Module.Param0WriteEnabled=true
    Module.Param0WriteEnabled1=true
    Module.Param1WriteEnabled=true
    Module.Param1WriteEnabled1=false
    Module.Param2WriteEnabled=true
    Module.Param2WriteEnabled1=false
    Module.Param3WriteEnabled=false
    Module.Param3WriteEnabled1=false
    Write Parameter Index 1=true
   )
  ]
  particleUpdateScript [
   &InheritSourceMovement(Applied Position Falloff Distance=100, Applied Position Scale=<0.9 0.9 1>, ApplyForce=false, ApplyPosition=true, ApplyPositionFalloff=false) {enabled=false}
   &SolveForcesAndVelocity(Clamp Velocity=false, Limit Acceleration=false)
   &FN_UpdateRibbon(Detach Locked Particle=true, Enable Skeletal Position=true, Sampled Position Coordinate Space:ENiagaraCoordinateSpace=::World)
   &ScaleRibbonWidth(Ribbon Width Scale=&FloatFromCurve(FloatCurve=DataInterface() {propertyValues {Curve {keys [{interpMode=RCIM_Cubic, leaveTangent=-0.3625, tangentMode=RCTM_Break, time=0, value=1}, <at 0.9 0.25>]}}}))
   &ScaleColor(
    Linear Color Curve=DataInterface() {propertyValues {
      AlphaCurve {keys [<at 0 1>,<at 0.4763 0.4>,<at 0.7785 0>]}
      BlueCurve {keys [<at 0 8>,<at 0.4086 1>,<at 1 1>]}
      GreenCurve {keys [<at 0 3.7102>,<at 0.4086 0.8468>,<at 1 1>]}
      RedCurve {keys [<at 0 4.1337>,<at 0.4086 0.5938>,<at 1 1>]}
     }}
    Scale Mode=::NewEnumerator2
   )
   &ScaleColor(Scale Alpha=$System.Alpha, ScaleA=true, ScaleRGB=true) {module=ScaleColor001}
  ]
  rendererClasses [&NiagaraRibbonRendererProperties]
  renderers [{rendererClass=&NiagaraRibbonRendererProperties}]
 }
 RibbonWhite001 {
  emitterUpdateScript [
   &EmitterState()
   &FN_SpawnRibbon_v2(Number of Ribbons=2, Spawn Mode=::NewEnumerator0, SpawnRate:NiagaraFloat=$System.SpawnRate)
  ]
  particleSpawnScript [
   &FN_SetSpawnGroup(Number of Spawn Groups=2)
   &FN_InitializeRibbon_v3(
    Color:LinearColor=&FNTimeOfDayColor(Night Color=<rgba 1 1 1 0.13>, Morning Color=<rgba 1 1 1 0.13>, Day Color=<rgba 1 1 1 0.13>, Evening Color=<rgba 1 1 1 0.13>)
    Custom Offset=true
    Enable Skeletal Position=true
    Lifetime=0.08
    Local Vector Offset:Vector3f=<0 1 0>
    Ribbon Width=25
   )
   &SkeletalMeshLocation(
    Apply Sampled Orientation to Mesh Particles=true
    Bone / Socket Index=&SpawnGroup()
    Bone Sampling Mode=::NewEnumerator4
    Inherit Velocity=false
    Interpolate Mesh Sampling=false
    MaskSpawnGroup=false
    Sampled Position Offset Coordinate Space:ENiagaraMeshTransforms=::NewEnumerator3
    Skeletal Mesh=$System.Skeletal Mesh
   )
   &DynamicMaterialParameters(
    Index 0 Param 1=&MaskFloatBySpawnGroup(Masked Float=1, Passthrough Float=0.8)
    Index 0 Param 2=&MaskFloatBySpawnGroup(Masked Float=0.4, Passthrough Float=3)
    Index 0 Param 3=1
    Index 1 Param 1=&Lerp_Float(A=5, B=10, Alpha=$System.Dive)
    Module.Param0WriteEnabled=true
    Module.Param0WriteEnabled1=true
    Module.Param1WriteEnabled=true
    Module.Param1WriteEnabled1=false
    Module.Param2WriteEnabled=true
    Module.Param2WriteEnabled1=false
    Module.Param3WriteEnabled=false
    Module.Param3WriteEnabled1=false
    Write Parameter Index 1=true
   )
  ]
  particleUpdateScript [
   &SolveForcesAndVelocity(Clamp Velocity=false, Limit Acceleration=false)
   &FN_UpdateRibbon(Detach Locked Particle=true, Enable Skeletal Position=true, Sampled Position Coordinate Space:ENiagaraCoordinateSpace=::World)
   &ScaleRibbonWidth(Ribbon Width Scale=&FloatFromCurve(FloatCurve=DataInterface() {propertyValues {Curve {keys [{interpMode=RCIM_Cubic, leaveTangent=-0.3625, tangentMode=RCTM_Break, time=0, value=1}, <at 0.9 0.25>]}}}))
   &ScaleColor(
    Linear Color Curve=DataInterface() {propertyValues {
      AlphaCurve {keys [<at 0 1>,<at 0.4763 0.4>,<at 0.7785 0>]}
      BlueCurve {keys [<at 0 8>,<at 0.4086 1>,<at 1 1>]}
      GreenCurve {keys [<at 0 3.7102>,<at 0.4086 0.8468>,<at 1 1>]}
      RedCurve {keys [<at 0 4.1337>,<at 0.4086 0.5938>,<at 1 1>]}
     }}
    Scale Mode=::NewEnumerator2
   )
   &ScaleColor(Scale Alpha=$System.Alpha, ScaleA=true, ScaleRGB=true) {module=ScaleColor001}
  ]
  rendererClasses [&NiagaraRibbonRendererProperties]
  renderers [{rendererClass=&NiagaraRibbonRendererProperties}]
 }
 RibbonWhite002 {
  bEnabled=false
  emitterUpdateScript [
   &EmitterState()
   &FN_EmitterDivingAttributes_v1(Bone Index=1, SkeletalMesh:NiagaraDataInterfaceSkeletalMesh=$User.SkeletalMesh) {enabled=false}
   &FN_SpawnRibbon_v2(Spawn Mode=::NewEnumerator0, SpawnRate:NiagaraFloat=$System.SpawnRate)
  ]
  particleSpawnScript [
   &FN_SetSpawnGroup(Number of Spawn Groups=1)
   &FN_InitializeRibbon_v3(
    Color:LinearColor=&FNTimeOfDayColor(Night Color=<rgba 1 1 1 0.13>, Morning Color=<rgba 1 1 1 0.13>, Day Color=<rgba 1 1 1 0.13>, Evening Color=<rgba 1 1 1 0.13>)
    Custom Offset=true
    Enable Skeletal Position=true
    Lifetime=0.08
    Local Vector Offset:Vector3f=<0 1 0>
    Ribbon Width=25
   )
   &AddVelocity(
    Scale Added Velocity=&Lerp_Vector(A=<1 1 1>, B=<4 4 1>, Alpha=$System.Dive)
    Velocity=&FN_FrontEndVector3Modifier(Gameplay=<0 0 550>, Front End=<0 0 550>)
   ) {enabled=false}
   &AddVelocity(Velocity=&MakeVector(
     X=&SelectFloatFromArray(Direct Array Index=$Particles.SpawnGroup, Float Selection Array=DataInterface() {propertyValues {FloatData [100,-100]}})
     Y=0
     Z=0
    )) {enabled=false, module=AddVelocity001}
   &SkeletalMeshLocation(
    Apply Sampled Orientation to Mesh Particles=true
    Bone / Socket Index=1
    Bone Sampling Mode=::NewEnumerator4
    Inherit Velocity=false
    Interpolate Mesh Sampling=false
    MaskSpawnGroup=false
    Sampled Position Offset Coordinate Space:ENiagaraMeshTransforms=::NewEnumerator3
    Skeletal Mesh=$System.Skeletal Mesh
   )
   &DynamicMaterialParameters(
    Index 0 Param 1=&MaskFloatBySpawnGroup(Masked Float=1, Passthrough Float=0.8)
    Index 0 Param 2=&MaskFloatBySpawnGroup(Masked Float=0.4, Passthrough Float=3)
    Index 0 Param 3=1
    Index 1 Param 1=&Lerp_Float(A=5, B=10, Alpha=$System.Dive)
    Module.Param0WriteEnabled=true
    Module.Param0WriteEnabled1=true
    Module.Param1WriteEnabled=true
    Module.Param1WriteEnabled1=false
    Module.Param2WriteEnabled=true
    Module.Param2WriteEnabled1=false
    Module.Param3WriteEnabled=false
    Module.Param3WriteEnabled1=false
    Write Parameter Index 1=true
   )
  ]
  particleUpdateScript [
   &InheritSourceMovement(Applied Position Falloff Distance=100, Applied Position Scale=<0.9 0.9 1>, ApplyForce=false, ApplyPosition=true, ApplyPositionFalloff=false) {enabled=false}
   &SolveForcesAndVelocity(Clamp Velocity=false, Limit Acceleration=false)
   &FN_UpdateRibbon(Detach Locked Particle=true, Enable Skeletal Position=true, Sampled Position Coordinate Space:ENiagaraCoordinateSpace=::World)
   &ScaleRibbonWidth(Ribbon Width Scale=&FloatFromCurve(FloatCurve=DataInterface() {propertyValues {Curve {keys [{interpMode=RCIM_Cubic, leaveTangent=-0.3625, tangentMode=RCTM_Break, time=0, value=1}, <at 0.9 0.25>]}}}))
   &ScaleColor(
    Linear Color Curve=DataInterface() {propertyValues {
      AlphaCurve {keys [<at 0 1>,<at 0.4763 0.4>,<at 0.7785 0>]}
      BlueCurve {keys [<at 0 8>,<at 0.4086 1>,<at 1 1>]}
      GreenCurve {keys [<at 0 3.7102>,<at 0.4086 0.8468>,<at 1 1>]}
      RedCurve {keys [<at 0 4.1337>,<at 0.4086 0.5938>,<at 1 1>]}
     }}
    Scale Mode=::NewEnumerator2
   )
   &ScaleColor(Scale Alpha=$System.Alpha, ScaleA=true, ScaleRGB=true) {module=ScaleColor001}
  ]
  rendererClasses [&NiagaraRibbonRendererProperties]
  renderers [{rendererClass=&NiagaraRibbonRendererProperties}]
 }
 RibbonWhite004 {
  emitterUpdateScript [
   &EmitterState()
   &FN_SpawnRibbon_v2(Number of Ribbons=2, Spawn Mode=::NewEnumerator0, SpawnRate:NiagaraFloat=$System.SpawnRate)
  ]
  particleSpawnScript [
   &FN_SetSpawnGroup(Number of Spawn Groups=2)
   &FN_InitializeRibbon_v3(
    Color:LinearColor=$System.Linear Color
    Custom Offset=true
    Enable Skeletal Position=true
    Local Vector Offset:Vector3f=<0 1 0>
    Ribbon Width=25
   )
   &SkeletalMeshLocation(
    Apply Sampled Orientation to Mesh Particles=true
    Bone / Socket Index=&SpawnGroup()
    Bone Sampling Mode=::NewEnumerator4
    Inherit Velocity=false
    Interpolate Mesh Sampling=false
    MaskSpawnGroup=false
    Sampled Position Offset Coordinate Space:ENiagaraMeshTransforms=::NewEnumerator3
    Skeletal Mesh=$System.Skeletal Mesh
   )
   &DynamicMaterialParameters(
    Index 0 Param 1=&MaskFloatBySpawnGroup(Masked Float=1, Passthrough Float=0.8)
    Index 0 Param 2=&MaskFloatBySpawnGroup(Masked Float=0.4, Passthrough Float=3)
    Index 0 Param 3=1
    Index 1 Param 1=&Lerp_Float(A=5, B=10, Alpha=$System.Dive)
    Module.Param0WriteEnabled=true
    Module.Param0WriteEnabled1=true
    Module.Param1WriteEnabled=true
    Module.Param1WriteEnabled1=false
    Module.Param2WriteEnabled=true
    Module.Param2WriteEnabled1=false
    Module.Param3WriteEnabled=false
    Module.Param3WriteEnabled1=false
    Write Parameter Index 1=true
   )
  ]
  particleUpdateScript [
   &SolveForcesAndVelocity(Clamp Velocity=false, Limit Acceleration=false)
   &FN_UpdateRibbon(Detach Locked Particle=true, Enable Skeletal Position=true, Sampled Position Coordinate Space:ENiagaraCoordinateSpace=::World)
   &ScaleRibbonWidth(Ribbon Width Scale=&FloatFromCurve(FloatCurve=DataInterface() {propertyValues {Curve {keys [{interpMode=RCIM_Cubic, leaveTangent=-0.3625, tangentMode=RCTM_Break, time=0, value=1}, <at 0.9 0.25>]}}}))
   &ScaleColor(
    Linear Color Curve=DataInterface() {propertyValues {
      AlphaCurve {keys [<at 0 1>,<at 0.4763 0.4>,<at 0.7785 0>]}
      BlueCurve {keys [<at 0 8>,<at 0.4086 1>,<at 1 1>]}
      GreenCurve {keys [<at 0 3.7102>,<at 0.4086 0.8468>,<at 1 1>]}
      RedCurve {keys [<at 0 4.1337>,<at 0.4086 0.5938>,<at 1 1>]}
     }}
    Scale Mode=::NewEnumerator2
   )
   &ScaleColor(Scale Alpha=$System.Alpha, ScaleA=true, ScaleRGB=true) {module=ScaleColor001}
  ]
  rendererClasses [&NiagaraRibbonRendererProperties]
  renderers [{rendererClass=&NiagaraRibbonRendererProperties}]
 }
 SpeedLines002 {
  bEnabled=false
  emitterUpdateScript [
   &EmitterState()
   &SpawnBurst_Instantaneous(Spawn Count=2, Spawn Time=0.3, Use Spawn Probability=false, UseLoopCountLimit=false)
   &SpawnRate(SpawnRate=20, Use Spawn Probability=false) {enabled=false}
  ]
  particleSpawnScript [
   &FN_SetSpawnGroup(Number of Spawn Groups=2)
   &InitializeParticle(
    Color=<rgba 0.4427 0.554 0.8646 0.15>
    Lifetime=3.5
    Mesh Scale Mode=::NewEnumerator3
    Mesh Uniform Scale=0.18
    Position Offset Coordinate Space=::Local
    Sprite Size Mode=::NewEnumerator3
    UsePositionOffset=true
    Write Material Random=true
   )
   &InitialMeshOrientation(
    Mesh Axis To Orient=<0 0 1>
    Mesh Orientation Mode=::NewEnumerator2
    Rotation=&RandomRangeVector(Minimum=<0 -1 0.1>, Maximum=<-0.1 1 -0.1>)
    Rotation Coordinate Space=::NewEnumerator2
    Use Rotation Vector=false
   ) {enabled=false}
   &SystemLocation(MaskSpawnGroup=false) {enabled=false}
   &DynamicMaterialParameters(
    Index 0 Param 2=0.3
    Index 0 Param 3=0.7
    Module.Param0WriteEnabled=false
    Module.Param1WriteEnabled=true
    Module.Param2WriteEnabled=true
    Module.Param3WriteEnabled=false
   ) {module=DynamicMaterialParameters001}
  ]
  particleUpdateScript [
   &ParticleState()
   &SkeletalMeshLocation(
    Apply Sampled Orientation to Mesh Particles=true
    Bone / Socket Index=&SpawnGroup()
    Bone Sampling Mode=::NewEnumerator4
    Inherit Velocity=false
    Interpolate Mesh Sampling=false
    MaskSpawnGroup=false
    Sampled Position Offset=<0 0 20>
    Sampled Position Offset Coordinate Space:ENiagaraMeshTransforms=::NewEnumerator3
    Skeletal Mesh=$System.Skeletal Mesh
   )
   &DynamicMaterialParameters(
    Index 0 Param 1=&FloatFromCurve(
     FloatCurve=DataInterface() {propertyValues {Curve {keys [
         {arriveTangent=9.0725, interpMode=RCIM_Cubic, leaveTangent=9.0725, tangentMode=RCTM_User, time=0, value=0}
         {arriveTangent=-2.5249, interpMode=RCIM_Cubic, leaveTangent=-2.5249, tangentMode=RCTM_User, time=0.1408, value=0.9847}
        ]}}}
     Scale Curve=8
    )
    Module.Param0WriteEnabled=true
    Module.Param1WriteEnabled=false
    Module.Param2WriteEnabled=false
    Module.Param3WriteEnabled=true
   )
   &ScaleColor(
    Scale Alpha=&FloatFromCurve(FloatCurve=DataInterface() {propertyValues {Curve {keys [
         {arriveTangent=19.4025, interpMode=RCIM_Cubic, leaveTangent=19.4025, tangentMode=RCTM_User, time=0, value=0}
         {arriveTangent=2, interpMode=RCIM_Cubic, leaveTangent=2, tangentMode=RCTM_User, time=0.1, value=1}
         {arriveTangent=-1.4286, interpMode=RCIM_Cubic, leaveTangent=-1.4286, tangentMode=RCTM_User, time=0.8, value=1}
         {arriveTangent=-0.0405, interpMode=RCIM_Cubic, leaveTangent=-0.0405, tangentMode=RCTM_User, time=1, value=0}
        ]}}})
    ScaleA=true
    ScaleRGB=true
   )
   &InheritSourceMovement(ApplyForce=false, ApplyPosition=true, ApplyPositionFalloff=false) {enabled=false}
   &SolveForcesAndVelocity(Clamp Velocity=false, Limit Acceleration=false)
   &ScaleMeshSize(Scale Factor=&Multiply_VectorByFloat(
     Vector=<1 1 0.2>
     Float=&FloatFromCurve(
      FloatCurve=DataInterface() {propertyValues {Curve {keys [
          {arriveTangent=4.8321, interpMode=RCIM_Cubic, leaveTangent=4.8321, tangentMode=RCTM_User, time=0, value=0.35}
          {interpMode=RCIM_Cubic, tangentMode=RCTM_User, time=0.2, value=1}
         ]}}}
      Scale Curve=1.7
     )
    ))
  ]
  rendererClasses [&NiagaraMeshRendererProperties]
  renderers [{rendererClass=&NiagaraMeshRendererProperties, rendererIndex=0}]
 }
}
summary {
 systemName=NS_Emote_GarlicWhisk_Trail
 emitters [
  {emitterName=RibbonWhite004, rendererClasses [&NiagaraRibbonRendererProperties]}
  {bEnabled=false, emitterName=RibbonWhite, rendererClasses [&NiagaraRibbonRendererProperties]}
  {emitterName=RibbonWhite001, rendererClasses [&NiagaraRibbonRendererProperties]}
  {bEnabled=false, emitterName=RibbonWhite002, rendererClasses [&NiagaraRibbonRendererProperties]}
  {bEnabled=false, emitterName=SpeedLines002, rendererClasses [&NiagaraMeshRendererProperties]}
 ]
 userVariables [
  User.bIsFrontEnd {type=NiagaraBool}
  User.bIsFrontEndPreview {type=NiagaraBool}
  User.RibbonIDOffset {type=NiagaraInt32}
  User.SkeletalMesh {
   type=NiagaraDataInterfaceSkeletalMesh
   defaultValue {
    struct=NiagaraExt_VariableValue_DataInterface
    value {dataInterface=&self:NiagaraDataInterfaceSkeletalMesh_1}
   }
  }
 ]
}
